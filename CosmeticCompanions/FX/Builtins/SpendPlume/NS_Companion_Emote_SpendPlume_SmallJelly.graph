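# Niagara system for the SpendPlume companion emote's small jellyfish, producing bubbles and bubble pops.
# SmallJelly_Bubbles streams from a skeletal mesh at a rate scaled by the timing parameter's magnitude, whose sign also toggles sideways drift.
# FXJelly_SpawnBubbles bursts around the jellyfishmain bone and records death positions, while BubblePops spawns ring-and-droplet sprites from system death data.
# Both bubble emitters blend world/local motion using owner velocity; brightness follows time of day.
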NiagaraSystem /CosmeticCompanions/FX/Builtins/SpendPlume/NS_Companion_Emote_SpendPlume_SmallJelly
dependencies {
 usedDataInterfaces [
  &NiagaraDataInterfaceCurve
  &NiagaraDataInterfaceVector2DCurve
  &NiagaraDataInterfaceVectorCurve
  &NiagaraDataInterfaceSkeletalMesh
  &NiagaraDataInterfaceArrayColor
 ]
 usedDynamicInputs []
 usedModules [
  &self:SystemSpawnScript.SystemScriptSource.SystemScriptGraph.NiagaraNodeAssignment_0.SetVariables_#1
  &self:SystemSpawnScript.SystemScriptSource.SystemScriptGraph.NiagaraNodeAssignment_1.SetVariables_#2
  &SystemState
  &BlendSpaceHelper
 ]
}
emitters {
 BubblePops {
  emitterUpdateScript [
   &EmitterState(Life Cycle Mode=::NewEnumerator1, Loop Behavior=::NewEnumerator1, Loop Duration Mode=::NewEnumerator1, UseLoopDelay=false)
   &SpawnPerFrame(Spawn:NiagaraBool=$System.DeathSpawn, Spawn Probability=0.8, Use Spawn Probability=false)
   &SpawnPerFrame(
    Spawn:NiagaraBool=$System.DeathSpawn
    Spawn Count:NiagaraInt32=&RandomRangeInt(Minimum=2, Maximum=3)
    Spawn Group=1
    Use Spawn Probability=false
   ) {enabled=false, module=SpawnPerFrame001}
  ]
  particleSpawnScript [
   /SetVariables(Particles.VisibilityTag=&SpawnGroup()) {isSetParameters=true}
   &InitializeParticle(
    Color=<rgba 1 1 1 0.5>
    Lifetime=0.6
    Mesh Scale Mode=::NewEnumerator3
    Position=&TransformPosition(Position=$System.DeathPosition, Source Space=::Local, Destination Space=::World)
    Position Mode=::NewEnumerator1
    Sprite Rotation Mode=::NewEnumerator2
    Sprite Size Mode=::NewEnumerator3
    Uniform Sprite Size=&RandomRangeFloat(Minimum=10, Maximum=12)
    UsePositionOffset=false
    Write Material Random=true
   )
   &DynamicMaterialParameters(
    Index 0 Param 1=5
    Index 0 Param 2=1
    Index 0 Param 3=1
    Index 0 Param 4=1
    Index 1 Param 1=6.667
    Index 1 Param 3=1
    Index 1 Param 4=1
    Module.Index0Float0Write=true
    Module.Index0Float1Write=true
    Module.Index0Float2Write=true
    Module.Index0Float3Write=true
    Module.Index1Float0Write=true
    Module.Index1Float1Write=true
    Module.Index1Float2Write=true
    Module.Index1Float3Write=true
    Write Parameter Index 1=::NewEnumerator1
   )
   /SetVariables(
    Particles.Drops Lifetime:NiagaraFloat=&RandomRangeFloat(Minimum=0.3, Maximum=0.2)
    Particles.Ring Lifetime=0.1
   ) {isSetParameters=true}
   &AddVelocity(
    Velocity=&RandomRangeVector(Minimum=<-15 -15 5>, Maximum=<15 15 21>)
    Velocity Speed Scale=&MaskFloatBySpawnGroup(Masked Float=0, Passthrough Float=1.5)
   )
   /SetVariables(Particles.Drops Size:Vector2f=&RandomRangeVector2D(Minimum=<0.5 2>, Maximum=<1 3>)) {isSetParameters=true}
   &CameraOffset(Camera Offset Amount=&RandomRangeFloat(Minimum=3, Maximum=9))
  ]
  particleUpdateScript [
   &ParticleState()
   /SetVariables(
    Particles.Update Drops Age:NiagaraFloat=&Divide_Float(A=$Particles.Age, B=$Particles.Initial.Drops Lifetime)
    Particles.Update Ring Age:NiagaraFloat=&Divide_Float(A=$Particles.Age, B=$Particles.Initial.Ring Lifetime)
   ) {isSetParameters=true}
   /SetVariables(
    Particles.Update Drops Dynamic Parameters:Vector4f=<0 0 1 1>
    Particles.Update Drops Linear Color:LinearColor=&MakeLinearColorFromVectorAndFloat(
     "Vector (RGB)"=<1 1 1>
     "Float (Alpha)"=&FloatFromCurve(
      FloatCurve=DataInterface() {propertyValues {Curve {keys [<at 0.2917 1 RCIM_Cubic>,<at 1 0>]}}}
      CurveIndex=$Particles.Update Drops Age
      Scale Curve=&FNTimeOfDayScalar(Night Scale=0.75, Morning Scale=0.9, Evening Scale=0.8)
     )
    )
    Particles.Update Drops Size:Vector2f=&Multiply_Vector2D(
     A=$Particles.Initial.Drops Size
     B=&Vector2DFromFloat(Value=&FloatFromCurve(
       FloatCurve=DataInterface() {propertyValues {Curve {keys [
           {arriveTangent=1.3698, interpMode=RCIM_Cubic, leaveTangent=1.3698, tangentMode=RCTM_User, time=0, value=0.3}
           {tangentMode=RCTM_User, time=1, value=1}
          ]}}}
       CurveIndex=$Particles.Update Drops Age
      ))
    )
   ) {isSetParameters=true}
   &GravityForce(Gravity=&MaskVectorBySpawnGroup(Masked Vector=<0 0 0>, Passthrough Vector=&RandomRangeVector(Minimum=<0 0 -120>, Maximum=<0 0 -70>)))
   &ScaleSpriteSize(
    Scale Sprite Size Mode=::NewEnumerator0
    Uniform Scale Factor=&FloatFromCurve(
     FloatCurve=DataInterface() {propertyValues {Curve {keys [
         {arriveTangent=0.3236, interpMode=RCIM_Cubic, leaveTangent=0.3236, tangentMode=RCTM_User, time=0, value=0.7}
         {tangentMode=RCTM_User, time=1, value=1.2}
        ]}}}
     CurveIndex=$Particles.Update Ring Age
    )
   )
   &ScaleColor(
    Scale Alpha=&FloatFromCurve(
     FloatCurve=DataInterface() {propertyValues {Curve {keys [<at 0.2917 1 RCIM_Cubic>,<at 1 0>]}}}
     CurveIndex=$Particles.Update Ring Age
     Scale Curve=&Multiply_Float(A=0.4, B=&FNTimeOfDayScalar(Night Scale=0.5, Morning Scale=0.8, Evening Scale=0.8))
    )
    Scale RGB=&VectorFromFloat(Value=&FloatFromCurve(
      FloatCurve=DataInterface() {propertyValues {Curve {keys [
          <at 0 3 RCIM_Cubic>
          {arriveTangent=-0.9668, interpMode=RCIM_Cubic, leaveTangent=-0.9668, tangentMode=RCTM_User, time=0.3086, value=1.5}
          <at 1 1 RCIM_Cubic>
         ]}}}
      CurveIndex=$Particles.Update Ring Age
     ))
    ScaleA=true
    ScaleRGB=true
   )
   &ScaleColor(
    Scale Alpha=&FNTimeOfDayScalar(Night Scale=0.8, Morning Scale=0.9, Evening Scale=0.9)
    Scale RGB=&MakeVectorFromLinearColorRGB(Color=&SelectLinearColorFromArray(
      Array Sampling Mode=::NewEnumerator1
      Direct Array Index=$Engine.QualityLevel
      Color Selection Array=DataInterface() {propertyValues {ColorData [<rgba 0 0 0 1>, <rgba 1 1 1 1>]}}
      Random Seed=?
     ))
    ScaleA=true
    ScaleRGB=true
   ) {module=ScaleColor001}
   &SolveForcesAndVelocity(Clamp Velocity=false, Limit Acceleration=false)
  ]
  rendererClasses [&NiagaraSpriteRendererProperties]
  renderers [
   {rendererClass=&NiagaraSpriteRendererProperties}
   {rendererClass=&NiagaraSpriteRendererProperties, rendererIndex=1}
  ]
 }
 FXJelly_SpawnBubbles {
  emitterSpawnScript [/SetVariables(Emitter.IsTraversing Speed=100, Emitter.WalkForward Speed=300) {isSetParameters=true, module=SetVariables_#3}]
  emitterUpdateScript [
   &EmitterState(Life Cycle Mode=::NewEnumerator1, Loop Behavior=::NewEnumerator1, Loop Duration=2, UseLoopDelay=false)
   &SpawnBurst_Instantaneous(Spawn Count=35, Use Spawn Probability=false, UseLoopCountLimit=false)
  ]
  particleSpawnScript [
   &InitializeParticle(
    Lifetime Max=0.65
    Lifetime Min=0.2
    Lifetime Mode=::NewEnumerator1
    Mass Mode=::NewEnumerator3
    Sprite Rotation Mode=::NewEnumerator2
    Sprite Size Mode=::NewEnumerator1
    Uniform Sprite Size Max=12
    Uniform Sprite Size Min=3
    UsePositionOffset=false
    Write Material Random=true
   )
   &SkeletalMeshLocation(
    MaskSpawnGroup=false
    OverrideRandomSeed=false
    Skeletal Mesh=DataInterface() {propertyValues {
      DefaultMesh=null
      ExcludeBoneName=root
      OptionalComponentClass=null
      PreviewMesh=/CosmeticCompanions/Assets/Props/Emotes/SpendPlume/Meshes/SKM_SpendPlume_Gadget
      SkinningMode=SkinOnTheFly
      SoftSourceActor=null
      SourceMode=Default
      UvSetIndex=0
      WholeMeshLOD=-1
      bExcludeBone=false
      bReadDeformedGeometry=true
      bRequireCurrentFrameData=true
      ComponentTags []
      FilteredBones [jellyfishmain]
      FilteredSockets []
      MeshUserParameter {parameter {name=None, typeDefHandle {registeredTypeIndex=1}, varData []}}
      SamplingRegions []
     }}
   )
   &ShapeLocation(MaskSpawnGroup=false, Offset=<0 0 -20>, Offset Mode=::NewEnumerator0, Override Seed=false, Sphere Radius=20, UseOwnerScale=true)
   &CameraOffset(Camera Offset Amount=&RandomRangeFloat(Minimum=3, Maximum=9))
  ]
  particleUpdateScript [
   &ParticleState()
   &ScaleSpriteSize(
    Non-Uniform Curve Scale=&RandomRangeVector2D(Minimum=<1 1>, Maximum=<1 1.3>)
    Non-Uniform Curve Sprite Scale=DataInterface() {propertyValues {
      XCurve {
       defaultValue=0
       keys [
        {arriveTangent=20.4729, interpMode=RCIM_Cubic, leaveTangent=20.4729, leaveTangentWeight=0.2846, tangentMode=RCTM_User, time=0, value=0}
        {
         arriveTangent=-9.5481
         arriveTangentWeight=0.0421
         interpMode=RCIM_Cubic
         leaveTangent=-0.4416
         leaveTangentWeight=0.2379
         tangentMode=RCTM_Break
         time=0.1263
         value=1
        }
        {arriveTangent=-1.1446, interpMode=RCIM_Cubic, leaveTangent=-1.1446, time=0.8401, value=0.6963}
        {arriveTangent=-10.9842, arriveTangentWeight=0.1, interpMode=RCIM_Cubic, leaveTangent=-10.9842, tangentMode=RCTM_User, time=1, value=0}
       ]
      }
      YCurve {keys [
        {arriveTangent=20.4729, interpMode=RCIM_Cubic, leaveTangent=20.4729, tangentMode=RCTM_User, time=0, value=0}
        {arriveTangent=-9.5481, interpMode=RCIM_Cubic, leaveTangent=-0.4416, tangentMode=RCTM_Break, time=0.1263, value=1}
        {arriveTangent=-1.1446, interpMode=RCIM_Cubic, leaveTangent=-1.1446, time=0.8401, value=0.6963}
        {arriveTangent=-10.9842, interpMode=RCIM_Cubic, leaveTangent=-10.9842, tangentMode=RCTM_User, time=1, value=0}
       ]}
     }}
    Scale Sprite Size Mode=::NewEnumerator3
   )
   &DynamicMaterialParameters(
    Index 0 Param 1=&FloatFromCurve(FloatCurve=DataInterface() {propertyValues {Curve {keys [
         {arriveTangent=0.1706, interpMode=RCIM_Cubic, leaveTangent=0.1706, leaveTangentWeight=0.3081, tangentMode=RCTM_User, time=0, value=0}
         {arriveTangentWeight=0.3104, interpMode=RCIM_Cubic, leaveTangentWeight=0.0345, tangentMode=RCTM_User, time=0.9, value=0.2}
         {arriveTangent=3.2005, interpMode=RCIM_Cubic, leaveTangent=3.2005, tangentMode=RCTM_User, time=1, value=0.5}
        ]}}})
    Module.Index0Float0Write=true
    Module.Index0Float1Write=true
    Module.Index0Float2Write=true
    Module.Index0Float3Write=true
   )
   &ScaleColor(
    Scale Alpha=&Multiply_Float(
     A=1
     B=&FloatFromCurve(
      FloatCurve=DataInterface() {propertyValues {Curve {keys [
          <at 0.1014 0 RCIM_Cubic>
          {interpMode=RCIM_Cubic, leaveTangentWeight=0.3713, tangentMode=RCTM_User, time=0.5, value=1}
          {
           arriveTangent=-3.2389
           arriveTangentWeight=0.386
           interpMode=RCIM_Cubic
           leaveTangent=-3.2389
           leaveTangentWeight=0.1013
           tangentMode=RCTM_User
           time=0.8949
           value=0.5475
          }
          {arriveTangent=-1.4392, interpMode=RCIM_Cubic, leaveTangent=-1.4392, tangentMode=RCTM_User, time=1.0011, value=0.002}
         ]}}}
      Scale Curve=&Multiply_Float(A=1, B=$System.Bubbles TOD)
     )
    )
    ScaleA=true
    ScaleRGB=true
   )
   /SetVariables(Particles.Position=&AddVectorToPosition(
     Position=$Particles.Position
     Vector=&MakeVector(
      X=&FloatFromCurve(
       FloatCurve=DataInterface() {propertyValues {Curve {keys [<at 0 0>,<at 0.25 1>,<at 0.5 0>,<at 0.75 -1>,<at 1 0>]}}}
       Scale Curve=&RandomRangeFloat(Minimum=-1, Maximum=1)
      )
      Y=&FloatFromCurve(
       FloatCurve=DataInterface() {propertyValues {Curve {keys [<at 0 0>,<at 0.25 1>,<at 0.5 0>,<at 0.75 -1>,<at 1 0>]}}}
       Scale Curve=&RandomRangeFloat(Minimum=-0.5, Maximum=0.5)
      )
      Z=&Add_Float(
       A=&FloatFromCurve(
        FloatCurve=DataInterface() {propertyValues {Curve {keys [{arriveTangent=-1.0003, interpMode=RCIM_Cubic, leaveTangent=-1.0003, tangentMode=RCTM_User, time=0, value=0}, <at 0.2 1 RCIM_Cubic>]}}}
        Scale Curve=&RandomRangeFloat(Minimum=0.001, Maximum=0.002)
       )
       B=&FloatFromCurve(
        FloatCurve=DataInterface() {propertyValues {Curve {keys [<at 0.95 0 RCIM_Cubic>, {arriveTangent=403.0874, interpMode=RCIM_Cubic, leaveTangent=403.0874, tangentMode=RCTM_User, time=0.9966, value=5}]}}}
        Scale Curve=0.001
       )
      )
     )
    )) {isSetParameters=true}
   &CurlNoiseForce(
    Noise Frequency=75
    Noise Strength=&FloatFromCurve(
     FloatCurve=DataInterface() {propertyValues {Curve {keys [
         {arriveTangent=-39.1213, interpMode=RCIM_Cubic, leaveTangent=-39.1213, leaveTangentWeight=114.1625, tangentMode=RCTM_User, time=0, value=50}
         {arriveTangentWeight=0.3333, interpMode=RCIM_Cubic, leaveTangentWeight=53.6425, tangentMode=RCTM_User, time=1, value=25}
        ]}}}
     Scale Curve=&RandomRangeFloat(Minimum=5, Maximum=1)
    )
    Pan Noise=false
   )
   &GravityForce(Gravity=<0 0 200>)
   &SolveForcesAndVelocity(Clamp Velocity=false, Limit Acceleration=false)
   &BlendSpace("Position Blend (World <-> Local)":NiagaraFloat=&MakeCustomFloatFromBool(
     Bool=&SetBoolByFloatComparison(A=&VectorLength(Vector=$User.OwnerLocalVelocity), B=$Emitter.WalkForward Speed)
     True Float=1
     False Float=&Multiply_Float(
      A=1
      B=&FloatFromCurve(FloatCurve=DataInterface() {propertyValues {Curve {keys [<at 0.5 1>,<at 1 0.2>]}}})
     )
     Translation Contribution:NiagaraFloat=1
     Rotation Contribution:NiagaraFloat=1
    ))
   &self:WriteDeath_01(Position Array:NiagaraDataInterfaceArrayPosition=$System.DeathArray) {module=WriteDeath_01}
  ]
  rendererClasses [&NiagaraSpriteRendererProperties]
  renderers [{rendererClass=&NiagaraSpriteRendererProperties}]
 }
 SmallJelly_Bubbles {
  emitterSpawnScript [/SetVariables(Emitter.IsTraversing Speed=100, Emitter.WalkForward Speed=300) {isSetParameters=true, module=SetVariables_#3}]
  emitterUpdateScript [
   &EmitterState(Life Cycle Mode=::NewEnumerator1, Loop Duration=2, UseLoopDelay=false)
   &SpawnRate(
    SpawnRate=&Multiply_Float(
     A=50
     B=&MakeCustomFloatFromBool(
      Bool=&SetBoolByFloatComparison(A=$User.SmallJellyFish_Timing, B=0, Comparison Type=::NewEnumerator6)
      True Float=&Multiply_Float(A=$User.SmallJellyFish_Timing, B=-1)
      False Float=$User.SmallJellyFish_Timing
     )
    )
    Use Spawn Probability=false
   )
  ]
  particleSpawnScript [
   &InitializeParticle(
    Lifetime Max=0.35
    Lifetime Min=0.12
    Lifetime Mode=::NewEnumerator1
    Mass Mode=::NewEnumerator3
    Sprite Rotation Mode=::NewEnumerator2
    Sprite Size Mode=::NewEnumerator1
    Uniform Sprite Size Max=8
    Uniform Sprite Size Min=1.5
    UsePositionOffset=false
    Write Material Random=true
   )
   &SkeletalMeshLocation(Bone Sampling Mode=::NewEnumerator4, MaskSpawnGroup=false, Skeletal Mesh=$System.SmallJelly SKM)
   &ShapeLocation(MaskSpawnGroup=false, Override Seed=false, Sphere Radius=5, UseOwnerScale=true)
   &CameraOffset(Camera Offset Amount=&RandomRangeFloat(Minimum=2, Maximum=5))
   &AddVelocity(Velocity=&MakeVector(
     X=0
     Y=&MakeCustomFloatFromBool(Bool=&SetBoolByFloatComparison(A=$User.SmallJellyFish_Timing, B=0), True Float=-50, False Float=0)
     Z=0
    ))
  ]
  particleUpdateScript [
   &ParticleState()
   &ScaleSpriteSize(
    Non-Uniform Curve Scale=&RandomRangeVector2D(Minimum=<1 1>, Maximum=<1 1.3>)
    Non-Uniform Curve Sprite Scale=DataInterface() {propertyValues {
      XCurve {
       defaultValue=0
       keys [
        {arriveTangent=4.31, interpMode=RCIM_Cubic, leaveTangent=4.31, tangentMode=RCTM_User, time=0, value=0}
        <at 0.8537 0.9403 RCIM_Cubic>
        {arriveTangent=-10.9842, arriveTangentWeight=0.1, interpMode=RCIM_Cubic, leaveTangent=-10.9842, tangentMode=RCTM_User, time=1, value=0}
       ]
      }
      YCurve {keys [
        {arriveTangent=4.31, interpMode=RCIM_Cubic, leaveTangent=4.31, tangentMode=RCTM_User, time=0, value=0}
        <at 0.8537 0.9403 RCIM_Cubic>
        {arriveTangent=-10.9842, interpMode=RCIM_Cubic, leaveTangent=-10.9842, tangentMode=RCTM_User, time=1, value=0}
       ]}
     }}
    Scale Sprite Size Mode=::NewEnumerator3
   )
   &DynamicMaterialParameters(
    Index 0 Param 1=&FloatFromCurve(FloatCurve=DataInterface() {propertyValues {Curve {keys [
         {arriveTangent=0.1706, interpMode=RCIM_Cubic, leaveTangent=0.1706, leaveTangentWeight=0.3081, tangentMode=RCTM_User, time=0, value=0}
         {arriveTangentWeight=0.3104, interpMode=RCIM_Cubic, leaveTangentWeight=0.0345, tangentMode=RCTM_User, time=0.9, value=0.2}
         {arriveTangent=3.2005, interpMode=RCIM_Cubic, leaveTangent=3.2005, tangentMode=RCTM_User, time=1, value=0.5}
        ]}}})
    Module.Index0Float0Write=true
    Module.Index0Float1Write=true
    Module.Index0Float2Write=true
    Module.Index0Float3Write=true
   )
   &ScaleColor(
    Scale Alpha=&FloatFromCurve(
     FloatCurve=DataInterface() {propertyValues {Curve {keys [
         <at 0.0545 0 RCIM_Cubic>
         {interpMode=RCIM_Cubic, leaveTangentWeight=0.3713, tangentMode=RCTM_User, time=0.5, value=1}
         {
          arriveTangent=-3.2389
          arriveTangentWeight=0.386
          interpMode=RCIM_Cubic
          leaveTangent=-3.2389
          leaveTangentWeight=0.1013
          tangentMode=RCTM_User
          time=0.8949
          value=0.5475
         }
         {arriveTangent=-1.4392, interpMode=RCIM_Cubic, leaveTangent=-1.4392, tangentMode=RCTM_User, time=1.0011, value=0.002}
        ]}}}
     Scale Curve=&Multiply_Float(A=1, B=$System.Bubbles TOD)
    )
    ScaleA=true
    ScaleRGB=true
   )
   &CurlNoiseForce(
    Noise Frequency=75
    Noise Strength=&FloatFromCurve(
     FloatCurve=DataInterface() {propertyValues {Curve {keys [
         {arriveTangent=-39.1213, interpMode=RCIM_Cubic, leaveTangent=-39.1213, leaveTangentWeight=114.1625, tangentMode=RCTM_User, time=0, value=50}
         {arriveTangentWeight=0.3333, interpMode=RCIM_Cubic, leaveTangentWeight=53.6425, tangentMode=RCTM_User, time=1, value=25}
        ]}}}
     Scale Curve=&RandomRangeFloat(Minimum=3, Maximum=10)
    )
    Pan Noise=false
   )
   &GravityForce(Gravity=&Add_Vector(
     A=&MakeVector(
      X=0
      Y=&MakeCustomFloatFromBool(Bool=&SetBoolByFloatComparison(A=$User.SmallJellyFish_Timing, B=0), True Float=-100, False Float=0)
      Z=0
     )
     B=&Multiply_Vector(
      A=&RandomRangeVector(
       Minimum=&VectorFromCurve(
        VectorCurve=DataInterface() {propertyValues {
          XCurve {keys [<at 0 0>,<at 1 0>]}
          YCurve {keys [<at 0 0>,<at 1 0>]}
          ZCurve {keys [
            {interpMode=RCIM_Cubic, tangentMode=RCTM_User, time=0, value=10}
            {arriveTangent=88.4209, interpMode=RCIM_Cubic, leaveTangent=88.4209, tangentMode=RCTM_User, time=0.75, value=35}
           ]}
         }}
        Scale Curve=<1 1 3>
       )
       Maximum=&VectorFromCurve(
        VectorCurve=DataInterface() {propertyValues {
          XCurve {keys [<at 0 0>,<at 1 0>]}
          YCurve {keys [<at 0 0>,<at 1 0>]}
          ZCurve {keys [
            {interpMode=RCIM_Cubic, tangentMode=RCTM_User, time=0, value=10}
            {arriveTangent=133.5137, interpMode=RCIM_Cubic, leaveTangent=133.5137, tangentMode=RCTM_User, time=0.75, value=70}
           ]}
         }}
        Scale Curve=<1 1 7>
       )
      )
      B=&MakeVector(
       X=0
       Y=0
       Z=&FloatFromCurve(
        FloatCurve=DataInterface() {propertyValues {Curve {defaultValue=1, keys [{tangentMode=RCTM_User, time=0.6542, value=0.5}, {tangentMode=RCTM_User, time=0.7019, value=1}]}}}
        Scale Curve=&RandomRangeFloat(Minimum=2, Maximum=3)
       )
      )
     )
    ))
   &SolveForcesAndVelocity(Clamp Velocity=false, Limit Acceleration=false)
   &BlendSpace("Position Blend (World <-> Local)":NiagaraFloat=&MakeCustomFloatFromBool(
     Bool=&SetBoolByFloatComparison(A=&VectorLength(Vector=$User.OwnerLocalVelocity), B=$Emitter.WalkForward Speed)
     True Float=1
     False Float=&Multiply_Float(
      A=1
      B=&FloatFromCurve(FloatCurve=DataInterface() {propertyValues {Curve {keys [<at 0.5 1>,<at 1 0>]}}})
     )
     Translation Contribution:NiagaraFloat=1
     Rotation Contribution:NiagaraFloat=1
    ))
  ]
  rendererClasses [&NiagaraSpriteRendererProperties]
  renderers [{rendererClass=&NiagaraSpriteRendererProperties}]
 }
}
summary {
 systemName=NS_Companion_Emote_SpendPlume_SmallJelly
 emitters [
  {emitterName=SmallJelly_Bubbles, rendererClasses [&NiagaraSpriteRendererProperties]}
  {emitterName=FXJelly_SpawnBubbles, rendererClasses [&NiagaraSpriteRendererProperties]}
  {emitterName=BubblePops, rendererClasses [&NiagaraSpriteRendererProperties]}
 ]
 userVariables [
  User.OwnerLocalVelocity {type=Vector3f, description="", defaultValue {struct=Vector3f, value=<0 0 0>}}
  User.SmallJellyFish_Timing {type=NiagaraFloat, defaultValue=0, description=""}
 ]
}
