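# Niagara system for the jellyfish SpendPlume companion cosmetic effect, with bubble trails, a spawn burst and pop sprites.
# Bubble emitters spawn from JellyFish emitter particles via particle reads: trails follow the Jelly_Push curve, while SpawnBubbles fires briefly at start.
# Owner speed switches trail velocity and world/local position blending; alpha scales by time of day.
# BubblePops spawns at system-read death positions; BigJelly_Bubbles is disabled.
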NiagaraSystem /CosmeticCompanions/FX/Builtins/SpendPlume/NS_Companion_SpendPlume_Jellyfish
dependencies {
 usedDataInterfaces [
  &NiagaraDataInterfaceSpriteRendererInfo
  &NiagaraDataInterfaceCurve
  &NiagaraDataInterfaceParticleRead
  &NiagaraDataInterfaceVectorCurve
  &NiagaraDataInterfaceVector2DCurve
  &NiagaraDataInterfaceArrayColor
  &NiagaraDataInterfaceCamera
  &NiagaraDataInterfaceSkeletalMesh
  &NiagaraDataInterfaceArrayPosition
 ]
 usedDynamicInputs []
 usedModules [
  &self:SystemSpawnScript.SystemScriptSource.SystemScriptGraph.NiagaraNodeAssignment_0.SetVariables_#1
  &self:SystemSpawnScript.SystemScriptSource.SystemScriptGraph.NiagaraNodeAssignment_1.SetVariables_#2
  &self:SystemSpawnScript.SystemScriptSource.SystemScriptGraph.NiagaraNodeAssignment_3.SetVariables_#3
  &SystemState
  &self:ReadDeath
  &self:SystemSpawnScript.SystemScriptSource.SystemScriptGraph.NiagaraNodeAssignment_2.SetVariables_#4
  &BlendSpaceHelper
 ]
}
emitters {
 BigJelly_Bubbles {
  bEnabled=false
  emitterSpawnScript [/SetVariables(Emitter.IsTraversing Speed=100, Emitter.WalkForward Speed=300) {isSetParameters=true, module=SetVariables_#5}]
  emitterUpdateScript [
   &EmitterState(Life Cycle Mode=::NewEnumerator1, Loop Duration=2, UseLoopDelay=false)
   &SpawnRate(SpawnRate=&Multiply_Float(A=60, B=$User.JellyFish_Timing), Use Spawn Probability=false)
  ]
  particleSpawnScript [
   &InitializeParticle(
    Lifetime Max=0.6
    Lifetime Min=0.25
    Lifetime Mode=::NewEnumerator1
    Mass Mode=::NewEnumerator3
    Sprite Rotation Mode=::NewEnumerator2
    Sprite Size Mode=::NewEnumerator1
    Uniform Sprite Size Max=20
    Uniform Sprite Size Min=3
    UsePositionOffset=false
    Write Material Random=true
   )
   &SkeletalMeshLocation(Bone Sampling Mode=::NewEnumerator4, MaskSpawnGroup=false, Skeletal Mesh=$System.BigJelly SKM)
   &ShapeLocation(MaskSpawnGroup=false, Non Uniform Scale=<1 0 1>, Override Seed=false, Sphere Radius=25, UseOwnerScale=true)
   &CameraOffset(Camera Offset Amount=&RandomRangeFloat(Minimum=0, Maximum=-15))
   &AddVelocity(Velocity=&RandomRangeVector(Minimum=<0 -22 10>, Maximum=<0 -100 -10>))
  ]
  particleUpdateScript [
   &ParticleState()
   &ScaleSpriteSize(
    Non-Uniform Curve Scale=&RandomRangeVector2D(Minimum=<1 1>, Maximum=<1 1.3>)
    Non-Uniform Curve Sprite Scale=DataInterface() {propertyValues {
      XCurve {
       defaultValue=0
       keys [
        {arriveTangent=9.1133, interpMode=RCIM_Cubic, leaveTangent=9.1133, tangentMode=RCTM_User, time=0, value=0}
        <at 0.3158 0.8349 RCIM_Cubic>
        {arriveTangent=-0.4951, arriveTangentWeight=0.1, interpMode=RCIM_Cubic, leaveTangent=-0.4951, tangentMode=RCTM_User, time=1, value=0}
       ]
      }
      YCurve {keys [
        {arriveTangent=9.1133, interpMode=RCIM_Cubic, leaveTangent=9.1133, tangentMode=RCTM_User, time=0, value=0}
        <at 0.3159 0.8389 RCIM_Cubic>
        {arriveTangent=-0.4951, interpMode=RCIM_Cubic, leaveTangent=-0.4951, tangentMode=RCTM_User, time=1, value=0}
       ]}
     }}
    Scale Sprite Size Mode=::NewEnumerator3
   )
   &DynamicMaterialParameters(
    Index 0 Param 1=&FloatFromCurve(FloatCurve=DataInterface() {propertyValues {Curve {keys [
         {arriveTangent=0.1706, interpMode=RCIM_Cubic, leaveTangent=0.1706, leaveTangentWeight=0.3081, tangentMode=RCTM_User, time=0, value=0}
         {arriveTangentWeight=0.3104, interpMode=RCIM_Cubic, leaveTangentWeight=0.0345, tangentMode=RCTM_User, time=0.9, value=0.2}
         {arriveTangent=3.2005, interpMode=RCIM_Cubic, leaveTangent=3.2005, tangentMode=RCTM_User, time=1, value=0.5}
        ]}}})
    Module.Index0Float0Write=true
    Module.Index0Float1Write=true
    Module.Index0Float2Write=true
    Module.Index0Float3Write=true
   )
   &ScaleColor(
    Scale Alpha=&FloatFromCurve(
     FloatCurve=DataInterface() {propertyValues {Curve {keys [
         <at 0.0545 0 RCIM_Cubic>
         {interpMode=RCIM_Cubic, leaveTangentWeight=0.3713, tangentMode=RCTM_User, time=0.5, value=1}
         {
          arriveTangent=-3.2389
          arriveTangentWeight=0.386
          interpMode=RCIM_Cubic
          leaveTangent=-3.2389
          leaveTangentWeight=0.1013
          tangentMode=RCTM_User
          time=0.8949
          value=0.5475
         }
         {arriveTangent=-1.4392, interpMode=RCIM_Cubic, leaveTangent=-1.4392, tangentMode=RCTM_User, time=1.0011, value=0.002}
        ]}}}
     Scale Curve=&Multiply_Float(A=1, B=$System.Bubbles TOD)
    )
    ScaleA=true
    ScaleRGB=true
   )
   &CurlNoiseForce(
    Noise Frequency=75
    Noise Strength=&FloatFromCurve(
     FloatCurve=DataInterface() {propertyValues {Curve {keys [
         {
          arriveTangent=-188.2384
          interpMode=RCIM_Cubic
          leaveTangent=-188.2384
          leaveTangentWeight=114.1625
          tangentMode=RCTM_User
          time=0.0878
          value=50
         }
         {arriveTangent=-37.9818, interpMode=RCIM_Cubic, leaveTangent=-37.9818, tangentMode=RCTM_User, time=0.2324, value=35.0346}
         {arriveTangentWeight=0.3333, interpMode=RCIM_Cubic, leaveTangentWeight=53.6425, tangentMode=RCTM_User, time=1, value=25}
        ]}}}
     Scale Curve=&RandomRangeFloat(Minimum=10, Maximum=3)
    )
    Pan Noise=false
   )
   &GravityForce(Gravity=&RandomRangeVector(
     Minimum=&VectorFromCurve(
      VectorCurve=DataInterface() {propertyValues {
        XCurve {keys [<at 0 0>,<at 1 0>]}
        YCurve {keys [<at 0 0>,<at 1 0>]}
        ZCurve {keys [
          {interpMode=RCIM_Cubic, tangentMode=RCTM_User, time=0, value=10}
          {arriveTangent=88.4209, interpMode=RCIM_Cubic, leaveTangent=88.4209, tangentMode=RCTM_User, time=0.75, value=35}
         ]}
       }}
      Scale Curve=<1 1 3>
     )
     Maximum=&VectorFromCurve(
      VectorCurve=DataInterface() {propertyValues {
        XCurve {keys [<at 0 0>,<at 1 0>]}
        YCurve {keys [<at 0 0>,<at 1 0>]}
        ZCurve {keys [
          {interpMode=RCIM_Cubic, tangentMode=RCTM_User, time=0, value=10}
          {arriveTangent=133.5137, interpMode=RCIM_Cubic, leaveTangent=133.5137, tangentMode=RCTM_User, time=0.75, value=70}
         ]}
       }}
      Scale Curve=<1 1 7>
     )
    ))
   &PointForce(
    Force Origin=&AddVectorToPosition(Position=$Engine.Emitter.SimulationPosition, Vector=<0 -20 40>)
    Force Strength=15
    Module.Use Falloff=false
    Module.UseLinearForce=false
   )
   &Drag(
    Drag=&Multiply_Float(
     A=&RandomRangeFloat(Minimum=0.5, Maximum=1.5)
     B=&FloatFromCurve(FloatCurve=DataInterface() {propertyValues {Curve {
         defaultValue=1
         keys [{arriveTangent=4.0894, interpMode=RCIM_Cubic, leaveTangent=4.0894, tangentMode=RCTM_User, time=0, value=0.25}, <at 0.4 1 RCIM_Cubic>]
        }}})
    )
    Use Linear Drag=true
    Use Rotational Drag=true
   )
   &SolveForcesAndVelocity(Clamp Velocity=false, Limit Acceleration=false)
   &BlendSpace("Position Blend (World <-> Local)":NiagaraFloat=&MakeCustomFloatFromBool(
     Bool=&SetBoolByFloatComparison(A=&VectorLength(Vector=$User.OwnerLocalVelocity), B=$Emitter.WalkForward Speed)
     True Float=1
     False Float=&Multiply_Float(
      A=1
      B=&FloatFromCurve(FloatCurve=DataInterface() {propertyValues {Curve {keys [<at 0.5 1>,<at 1 0>]}}})
     )
     Translation Contribution:NiagaraFloat=1
     Rotation Contribution:NiagaraFloat=1
    ))
  ]
  rendererClasses [&NiagaraSpriteRendererProperties]
  renderers [{rendererClass=&NiagaraSpriteRendererProperties}]
 }
 BubblePops {
  emitterUpdateScript [
   &EmitterState(Life Cycle Mode=::NewEnumerator1, Loop Behavior=::NewEnumerator1, Loop Duration Mode=::NewEnumerator1, UseLoopDelay=false)
   &SpawnPerFrame(Spawn:NiagaraBool=$System.DeathSpawn, Spawn Probability=0.8, Use Spawn Probability=false)
   &SpawnPerFrame(
    Spawn:NiagaraBool=$System.DeathSpawn
    Spawn Count:NiagaraInt32=&RandomRangeInt(Minimum=2, Maximum=3)
    Spawn Group=1
    Use Spawn Probability=false
   ) {enabled=false, module=SpawnPerFrame001}
  ]
  particleSpawnScript [
   /SetVariables(Particles.VisibilityTag=&SpawnGroup()) {isSetParameters=true}
   &InitializeParticle(
    Color=<rgba 1 1 1 0.5>
    Lifetime=0.6
    Mesh Scale Mode=::NewEnumerator3
    Position=&TransformPosition(Position=$System.DeathPosition, Source Space=::Local, Destination Space=::World)
    Position Mode=::NewEnumerator1
    Sprite Rotation Mode=::NewEnumerator2
    Sprite Size Mode=::NewEnumerator3
    Uniform Sprite Size=&RandomRangeFloat(Minimum=10, Maximum=12)
    UsePositionOffset=false
    Write Material Random=true
   )
   &DynamicMaterialParameters(
    Index 0 Param 1=5
    Index 0 Param 2=1
    Index 0 Param 3=1
    Index 0 Param 4=1
    Index 1 Param 1=6.667
    Index 1 Param 3=1
    Index 1 Param 4=1
    Module.Index0Float0Write=true
    Module.Index0Float1Write=true
    Module.Index0Float2Write=true
    Module.Index0Float3Write=true
    Module.Index1Float0Write=true
    Module.Index1Float1Write=true
    Module.Index1Float2Write=true
    Module.Index1Float3Write=true
    Write Parameter Index 1=::NewEnumerator1
   )
   /SetVariables(
    Particles.Drops Lifetime:NiagaraFloat=&RandomRangeFloat(Minimum=0.3, Maximum=0.2)
    Particles.Ring Lifetime=0.1
   ) {isSetParameters=true}
   &AddVelocity(
    Velocity=&RandomRangeVector(Minimum=<-15 -15 5>, Maximum=<15 15 21>)
    Velocity Speed Scale=&MaskFloatBySpawnGroup(Masked Float=0, Passthrough Float=1.5)
   )
   /SetVariables(Particles.Drops Size:Vector2f=&RandomRangeVector2D(Minimum=<0.5 2>, Maximum=<1 3>)) {isSetParameters=true}
   &CameraOffset(Camera Offset Amount=&RandomRangeFloat(Minimum=3, Maximum=9))
  ]
  particleUpdateScript [
   &ParticleState()
   /SetVariables(
    Particles.Update Drops Age:NiagaraFloat=&Divide_Float(A=$Particles.Age, B=$Particles.Initial.Drops Lifetime)
    Particles.Update Ring Age:NiagaraFloat=&Divide_Float(A=$Particles.Age, B=$Particles.Initial.Ring Lifetime)
   ) {isSetParameters=true}
   /SetVariables(
    Particles.Update Drops Dynamic Parameters:Vector4f=<0 0 1 1>
    Particles.Update Drops Linear Color:LinearColor=&MakeLinearColorFromVectorAndFloat(
     "Vector (RGB)"=<1 1 1>
     "Float (Alpha)"=&FloatFromCurve(
      FloatCurve=DataInterface() {propertyValues {Curve {keys [<at 0.2917 1 RCIM_Cubic>,<at 1 0>]}}}
      CurveIndex=$Particles.Update Drops Age
      Scale Curve=&FNTimeOfDayScalar(Night Scale=0.75, Morning Scale=0.9, Evening Scale=0.8)
     )
    )
    Particles.Update Drops Size:Vector2f=&Multiply_Vector2D(
     A=$Particles.Initial.Drops Size
     B=&Vector2DFromFloat(Value=&FloatFromCurve(
       FloatCurve=DataInterface() {propertyValues {Curve {keys [
           {arriveTangent=1.3698, interpMode=RCIM_Cubic, leaveTangent=1.3698, tangentMode=RCTM_User, time=0, value=0.3}
           {tangentMode=RCTM_User, time=1, value=1}
          ]}}}
       CurveIndex=$Particles.Update Drops Age
      ))
    )
   ) {isSetParameters=true}
   &GravityForce(Gravity=&MaskVectorBySpawnGroup(Masked Vector=<0 0 0>, Passthrough Vector=&RandomRangeVector(Minimum=<0 0 -120>, Maximum=<0 0 -70>)))
   &ScaleSpriteSize(
    Scale Sprite Size Mode=::NewEnumerator0
    Uniform Scale Factor=&FloatFromCurve(
     FloatCurve=DataInterface() {propertyValues {Curve {keys [
         {arriveTangent=0.3236, interpMode=RCIM_Cubic, leaveTangent=0.3236, tangentMode=RCTM_User, time=0, value=0.7}
         {tangentMode=RCTM_User, time=1, value=1.2}
        ]}}}
     CurveIndex=$Particles.Update Ring Age
    )
   )
   &ScaleColor(
    Scale Alpha=&FloatFromCurve(
     FloatCurve=DataInterface() {propertyValues {Curve {keys [<at 0.2917 1 RCIM_Cubic>,<at 1 0>]}}}
     CurveIndex=$Particles.Update Ring Age
     Scale Curve=&Multiply_Float(A=0.4, B=&FNTimeOfDayScalar(Night Scale=0.5, Morning Scale=0.8, Evening Scale=0.8))
    )
    Scale RGB=&VectorFromFloat(Value=&FloatFromCurve(
      FloatCurve=DataInterface() {propertyValues {Curve {keys [
          <at 0 3 RCIM_Cubic>
          {arriveTangent=-0.9668, interpMode=RCIM_Cubic, leaveTangent=-0.9668, tangentMode=RCTM_User, time=0.3086, value=1.5}
          <at 1 1 RCIM_Cubic>
         ]}}}
      CurveIndex=$Particles.Update Ring Age
     ))
    ScaleA=true
    ScaleRGB=true
   )
   &ScaleColor(
    Scale Alpha=&FNTimeOfDayScalar(Night Scale=0.8, Morning Scale=0.9, Evening Scale=0.9)
    Scale RGB=&MakeVectorFromLinearColorRGB(Color=&SelectLinearColorFromArray(
      Array Sampling Mode=::NewEnumerator1
      Direct Array Index=$Engine.QualityLevel
      Color Selection Array=DataInterface() {propertyValues {ColorData [<rgba 0 0 0 1>, <rgba 1 1 1 1>]}}
      Random Seed=?
     ))
    ScaleA=true
    ScaleRGB=true
   ) {module=ScaleColor001}
   &SolveForcesAndVelocity(Clamp Velocity=false, Limit Acceleration=false)
  ]
  rendererClasses [&NiagaraSpriteRendererProperties]
  renderers [
   {rendererClass=&NiagaraSpriteRendererProperties}
   {rendererClass=&NiagaraSpriteRendererProperties, rendererIndex=1}
  ]
 }
 FXJelly_Bubbles {
  emitterSpawnScript [/SetVariables(Emitter.IsTraversing Speed=100, Emitter.WalkForward Speed=300) {isSetParameters=true, module=SetVariables_#5}]
  emitterUpdateScript [
   &EmitterState(Life Cycle Mode=::NewEnumerator1, Loop Duration=2, UseLoopDelay=false)
   &SpawnParticlesFromOtherEmitter(
    Attribute Reader=DataInterface() {propertyValues {EmitterBinding {bindingMode=Other, emitterName=JellyFish}}}
    Spawn Rate=&FloatFromCurve(FloatCurve=DataInterface() {propertyValues {Curve {keys [<at 0.5 1>,<at 1 0>]}}}, CurveIndex=$User.Jelly_Push, Scale Curve=15)
    Use Maximum Age=false
    Use Minimum Age=false
   )
  ]
  particleSpawnScript [
   &InitializeParticle(
    Lifetime Max=0.55
    Lifetime Min=0.2
    Lifetime Mode=::NewEnumerator1
    Mass Mode=::NewEnumerator3
    Sprite Rotation Mode=::NewEnumerator2
    UsePositionOffset=false
    Write Material Random=true
   )
   &SampleParticlesFromOtherEmitter(Apply Sampled RGB=false, Color Sampling=::NewEnumerator0, Sprite Size Sampling:ENiagara_DirectReadSamplingMode=::NewEnumerator0)
   &CameraOffset(Camera Offset Amount=-20)
   &ShapeLocation(MaskSpawnGroup=false, Override Seed=false, Sphere Radius=1, UseOwnerScale=true)
   /SetVariables(Particles.Position=&AddVectorToPosition(Position=$Particles.Position, Vector=<0 -2 -10>)) {isSetParameters=true}
   &AddVelocity(Velocity=&MakeCustomVectorFromBool(
     Bool=&SetBoolByFloatComparison(A=&VectorLength(Vector=$User.OwnerLocalVelocity), B=$Emitter.IsTraversing Speed)
     True Vector=<-5 -95 -20>
     False Vector=<5 -40 -90>
    ))
  ]
  particleUpdateScript [
   &ParticleState()
   &DynamicMaterialParameters(
    Index 0 Param 1=&FloatFromCurve(FloatCurve=DataInterface() {propertyValues {Curve {keys [
         {arriveTangent=0.1706, interpMode=RCIM_Cubic, leaveTangent=0.1706, leaveTangentWeight=0.3081, tangentMode=RCTM_User, time=0, value=0}
         {arriveTangentWeight=0.3104, interpMode=RCIM_Cubic, leaveTangentWeight=0.0345, tangentMode=RCTM_User, time=0.9, value=0.2}
         {arriveTangent=3.2005, interpMode=RCIM_Cubic, leaveTangent=3.2005, tangentMode=RCTM_User, time=1, value=0.5}
        ]}}})
    Module.Index0Float0Write=true
    Module.Index0Float1Write=true
    Module.Index0Float2Write=true
    Module.Index0Float3Write=true
   )
   /SetVariables(Particles.Position=&AddVectorToPosition(
     Position=$Particles.Initial.Position
     Vector=/Niagara/DynamicInputs/TypeConversions/MakeCustomPositionFromBool(
      Bool:NiagaraBool=&SetBoolByFloatComparison(A=&VectorLength(Vector=$User.OwnerLocalVelocity), B=$Emitter.IsTraversing Speed)
      True Position:NiagaraPosition=&MakeVector(
       X=&FloatFromCurve(
        FloatCurve=DataInterface() {propertyValues {Curve {keys [<at 0 0>,<at 0.25 1>,<at 0.5 0>,<at 0.75 -1>,<at 1 0>]}}}
        Scale Curve=&RandomRangeFloat(Minimum=-1, Maximum=1)
       )
       Y=&FloatFromCurve(
        FloatCurve=DataInterface() {propertyValues {Curve {keys [{arriveTangent=9.3254, interpMode=RCIM_Cubic, leaveTangent=9.3254, tangentMode=RCTM_User, time=0, value=0}, <at 0.2 1 RCIM_Cubic>]}}}
        Scale Curve=&RandomRangeFloat(Minimum=-0.5, Maximum=-1.2)
       )
       Z=&Add_Float(
        A=&FloatFromCurve(
         FloatCurve=DataInterface() {propertyValues {Curve {keys [<at 0 0>,<at 0.25 1>,<at 0.5 0>,<at 0.75 -1>,<at 1 0>]}}}
         Scale Curve=&RandomRangeFloat(Minimum=-1, Maximum=1)
        )
        B=&FloatFromCurve(FloatCurve=DataInterface() {propertyValues {Curve {keys [<at 0.95 0 RCIM_Cubic>, {arriveTangent=403.0874, interpMode=RCIM_Cubic, leaveTangent=403.0874, tangentMode=RCTM_User, time=0.9966, value=5}]}}})
       )
      )
      False Position:NiagaraPosition=&MakeVector(
       X=&FloatFromCurve(
        FloatCurve=DataInterface() {propertyValues {Curve {keys [<at 0 0>,<at 0.25 1>,<at 0.5 0>,<at 0.75 -1>,<at 1 0>]}}}
        Scale Curve=&RandomRangeFloat(Minimum=-1, Maximum=1)
       )
       Y=&Multiply_Float(
        A=&FloatFromCurve(
         FloatCurve=DataInterface() {propertyValues {Curve {keys [{arriveTangent=9.3254, interpMode=RCIM_Cubic, leaveTangent=9.3254, tangentMode=RCTM_User, time=0, value=0}, <at 0.2 1 RCIM_Cubic>]}}}
         Scale Curve=&RandomRangeFloat(Minimum=-2, Maximum=-4)
        )
        B=&FloatFromCurve(
         FloatCurve=DataInterface() {propertyValues {Curve {keys [<at 0 0>,<at 0.25 1>,<at 0.5 0>,<at 0.75 -1>,<at 1 0>]}}}
         Scale Curve=&RandomRangeFloat(Minimum=-0.5, Maximum=0.5)
        )
       )
       Z=&Add_Float(
        A=&FloatFromCurve(
         FloatCurve=DataInterface() {propertyValues {Curve {keys [{arriveTangent=9.3254, interpMode=RCIM_Cubic, leaveTangent=9.3254, tangentMode=RCTM_User, time=0, value=0}, <at 0.2 1 RCIM_Cubic>]}}}
         Scale Curve=&RandomRangeFloat(Minimum=-0.5, Maximum=-1.2)
        )
        B=&FloatFromCurve(FloatCurve=DataInterface() {propertyValues {Curve {keys [<at 0.95 0 RCIM_Cubic>, {arriveTangent=403.0874, interpMode=RCIM_Cubic, leaveTangent=403.0874, tangentMode=RCTM_User, time=0.9966, value=5}]}}})
       )
      )
     )
    )) {enabled=false, isSetParameters=true, module=SetVariables_#6}
   &CurlNoiseForce(
    Noise Frequency=75
    Noise Strength=&FloatFromCurve(
     FloatCurve=DataInterface() {propertyValues {Curve {keys [
         {arriveTangent=-39.1213, interpMode=RCIM_Cubic, leaveTangent=-39.1213, leaveTangentWeight=114.1625, tangentMode=RCTM_User, time=0, value=50}
         {arriveTangentWeight=0.3333, interpMode=RCIM_Cubic, leaveTangentWeight=53.6425, tangentMode=RCTM_User, time=1, value=25}
        ]}}}
     Scale Curve=&RandomRangeFloat(Minimum=2, Maximum=7)
    )
    Pan Noise=false
   )
   &Drag(
    Drag=&Multiply_Float(
     A=&RandomRangeFloat(Minimum=0.5, Maximum=1.5)
     B=&FloatFromCurve(FloatCurve=DataInterface() {propertyValues {Curve {defaultValue=1, keys [<at -0.0007 0>, {tangentMode=RCTM_User, time=0.1756, value=0.5471}, <at 1.0015 1.5>]}}})
    )
    Use Linear Drag=true
    Use Rotational Drag=true
   )
   &GravityForce(Gravity=&Multiply_Vector(
     A=&RandomRangeVector(
      Minimum=&VectorFromCurve(
       VectorCurve=DataInterface() {propertyValues {
         XCurve {keys [<at 0 0>,<at 1 0>]}
         YCurve {keys [<at 0 0>,<at 1 0>]}
         ZCurve {keys [
           {interpMode=RCIM_Cubic, tangentMode=RCTM_User, time=0, value=10}
           {arriveTangent=88.4209, interpMode=RCIM_Cubic, leaveTangent=88.4209, tangentMode=RCTM_User, time=0.75, value=35}
          ]}
        }}
       Scale Curve=<1 1 3>
      )
      Maximum=&VectorFromCurve(
       VectorCurve=DataInterface() {propertyValues {
         XCurve {keys [<at 0 0>,<at 1 0>]}
         YCurve {keys [<at 0 0>,<at 1 0>]}
         ZCurve {keys [
           {interpMode=RCIM_Cubic, tangentMode=RCTM_User, time=0, value=10}
           {arriveTangent=133.5137, interpMode=RCIM_Cubic, leaveTangent=133.5137, tangentMode=RCTM_User, time=0.75, value=70}
          ]}
        }}
       Scale Curve=<1 1 7>
      )
     )
     B=&MakeVector(
      X=0
      Y=0
      Z=&FloatFromCurve(
       FloatCurve=DataInterface() {propertyValues {Curve {defaultValue=1, keys [{tangentMode=RCTM_User, time=0.6542, value=0.5}, {tangentMode=RCTM_User, time=0.7019, value=1}]}}}
       Scale Curve=&RandomRangeFloat(Minimum=2, Maximum=3)
      )
     )
    ))
   &InheritVelocity(
    Inherited Velocity Amount Scale=&MakeCustomVectorFromBool(
     Bool=&SetBoolByFloatComparison(A=&VectorLength(Vector=$User.OwnerLocalVelocity), B=$Emitter.WalkForward Speed)
     True Vector=<0 0 0>
     False Vector=<-0.02 -0.02 0>
    )
    Speed Limit=true
    UseSpeedThreshold=false
   )
   &SolveForcesAndVelocity(Clamp Velocity=false, Limit Acceleration=false)
   &BlendSpace("Position Blend (World <-> Local)":NiagaraFloat=&MakeCustomFloatFromBool(
     Bool=&SetBoolByFloatComparison(A=&VectorLength(Vector=$User.OwnerLocalVelocity), B=$Emitter.WalkForward Speed)
     True Float=1
     False Float=&Multiply_Float(
      A=1
      B=&FloatFromCurve(FloatCurve=DataInterface() {propertyValues {Curve {keys [<at 0.5 1>,<at 1 0.2>]}}})
     )
     Translation Contribution:NiagaraFloat=1
     Rotation Contribution:NiagaraFloat=1
    ))
   /SetVariables(
    Particles.Linear Color=&MultiplyLinearColor(
     Linear Color=&MultiplyLinearColor(
      Linear Color=$Particles.Color
      Scale Factor=&MakeVector4_FromLinearColor(LinearColor=&MakeLinearColorFromVectorAndFloat("Vector (RGB)"=<1 1 1>, "Float (Alpha)"=$System.Bubbles TOD))
     )
     Scale Factor=&MakeVector4_FromLinearColor(LinearColor=&MakeLinearColorFromVectorAndFloat(
       "Vector (RGB)"=<1 1 1>
       "Float (Alpha)"=&Multiply_Float(
        A=$User.FX_Alpha
        B=&FloatFromCurve(FloatCurve=DataInterface() {propertyValues {Curve {keys [
             <at 0.3605 0 RCIM_Cubic>
             <at 0.4703 1 RCIM_Cubic>
             {arriveTangent=-5.5857, interpMode=RCIM_Cubic, leaveTangent=-5.5857, tangentMode=RCTM_User, time=1, value=0}
            ]}}})
       )
      ))
    )
    Particles.Vector 2D:Vector2f=&Multiply_Vector2D_ByFloat(
     Float=&FloatFromCurve(FloatCurve=DataInterface() {propertyValues {Curve {keys [
          <at 0.3605 0 RCIM_Cubic>
          <at 0.4703 1 RCIM_Cubic>
          {arriveTangent=-1.2784, interpMode=RCIM_Cubic, leaveTangent=-1.2784, tangentMode=RCTM_User, time=1, value=0}
         ]}}})
     Vector2D=$Particles.SpriteSize
    )
   ) {isSetParameters=true}
  ]
  rendererClasses [&NiagaraSpriteRendererProperties]
  renderers [{rendererClass=&NiagaraSpriteRendererProperties}]
 }
 FXJelly_Bubbles001 {
  emitterSpawnScript [/SetVariables(Emitter.IsTraversing Speed=100, Emitter.WalkForward Speed=300) {isSetParameters=true, module=SetVariables_#5}]
  emitterUpdateScript [
   &EmitterState(Life Cycle Mode=::NewEnumerator1, Loop Duration=2, UseLoopDelay=false)
   &SpawnParticlesFromOtherEmitter(
    Attribute Reader=DataInterface() {propertyValues {EmitterBinding {bindingMode=Other, emitterName=JellyFish}}}
    Spawn Rate=&FloatFromCurve(FloatCurve=DataInterface() {propertyValues {Curve {keys [<at 0.5 1>,<at 1 0>]}}}, CurveIndex=$User.Jelly_Push, Scale Curve=15)
    Use Maximum Age=false
    Use Minimum Age=false
   )
  ]
  particleSpawnScript [
   &InitializeParticle(
    Lifetime Max=0.55
    Lifetime Min=0.2
    Lifetime Mode=::NewEnumerator1
    Mass Mode=::NewEnumerator3
    Sprite Rotation Mode=::NewEnumerator2
    Sprite Size Mode=::NewEnumerator1
    Uniform Sprite Size Max=5
    Uniform Sprite Size Min=1.5
    UsePositionOffset=false
    Write Material Random=true
   )
   &SampleParticlesFromOtherEmitter(Color Sampling=::NewEnumerator0, Sprite Size Sampling:ENiagara_DirectReadSamplingMode=::NewEnumerator0)
   &ShapeLocation(
    MaskSpawnGroup=false
    Non Uniform Scale=<1 0.3 1>
    Offset=<0 -15 -5>
    Offset Mode=::NewEnumerator0
    Override Seed=false
    Sphere Radius=12
    UseOwnerScale=true
   )
   &CameraOffset(Camera Offset Amount=-22)
   /SetVariables(Particles.Position=&AddVectorToPosition(Position=$Particles.Position, Vector=<0 0 -7>)) {isSetParameters=true}
   &AddVelocity(Velocity=&MakeCustomVectorFromBool(
     Bool=&SetBoolByFloatComparison(A=&VectorLength(Vector=$User.OwnerLocalVelocity), B=$Emitter.IsTraversing Speed)
     True Vector=<-5 -95 -20>
     False Vector=<5 -40 -90>
    ))
  ]
  particleUpdateScript [
   &ParticleState()
   &ScaleSpriteSize(
    Non-Uniform Curve Scale=&RandomRangeVector2D(Minimum=<1 1>, Maximum=<1 1.3>)
    Non-Uniform Curve Sprite Scale=DataInterface() {propertyValues {
      XCurve {
       defaultValue=0
       keys [
        {arriveTangent=4.31, interpMode=RCIM_Cubic, leaveTangent=4.31, tangentMode=RCTM_User, time=0, value=0}
        <at 0.8537 0.9403 RCIM_Cubic>
        {arriveTangent=-10.9842, arriveTangentWeight=0.1, interpMode=RCIM_Cubic, leaveTangent=-10.9842, tangentMode=RCTM_User, time=1, value=0}
       ]
      }
      YCurve {keys [
        {arriveTangent=4.31, interpMode=RCIM_Cubic, leaveTangent=4.31, tangentMode=RCTM_User, time=0, value=0}
        <at 0.8537 0.9403 RCIM_Cubic>
        {arriveTangent=-10.9842, interpMode=RCIM_Cubic, leaveTangent=-10.9842, tangentMode=RCTM_User, time=1, value=0}
       ]}
     }}
    Scale Sprite Size Mode=::NewEnumerator3
   ) {enabled=false}
   &DynamicMaterialParameters(
    Index 0 Param 1=&FloatFromCurve(FloatCurve=DataInterface() {propertyValues {Curve {keys [
         {arriveTangent=0.1706, interpMode=RCIM_Cubic, leaveTangent=0.1706, leaveTangentWeight=0.3081, tangentMode=RCTM_User, time=0, value=0}
         {arriveTangentWeight=0.3104, interpMode=RCIM_Cubic, leaveTangentWeight=0.0345, tangentMode=RCTM_User, time=0.9, value=0.2}
         {arriveTangent=3.2005, interpMode=RCIM_Cubic, leaveTangent=3.2005, tangentMode=RCTM_User, time=1, value=0.5}
        ]}}})
    Module.Index0Float0Write=true
    Module.Index0Float1Write=true
    Module.Index0Float2Write=true
    Module.Index0Float3Write=true
   )
   /SetVariables(Particles.Position=&AddVectorToPosition(
     Position=$Particles.Position
     Vector=&MakeCustomPositionFromBool(
      Bool:NiagaraBool=&SetBoolByFloatComparison(A=&VectorLength(Vector=$User.OwnerLocalVelocity), B=$Emitter.IsTraversing Speed)
      True Position:NiagaraPosition=&MakeVector(
       X=&FloatFromCurve(
        FloatCurve=DataInterface() {propertyValues {Curve {keys [<at 0 0>,<at 0.25 1>,<at 0.5 0>,<at 0.75 -1>,<at 1 0>]}}}
        Scale Curve=&RandomRangeFloat(Minimum=-1, Maximum=1)
       )
       Y=&FloatFromCurve(
        FloatCurve=DataInterface() {propertyValues {Curve {keys [{arriveTangent=9.3254, interpMode=RCIM_Cubic, leaveTangent=9.3254, tangentMode=RCTM_User, time=0, value=0}, <at 0.2 1 RCIM_Cubic>]}}}
        Scale Curve=&RandomRangeFloat(Minimum=-0.5, Maximum=-1.2)
       )
       Z=&Add_Float(
        A=&FloatFromCurve(
         FloatCurve=DataInterface() {propertyValues {Curve {keys [<at 0 0>,<at 0.25 1>,<at 0.5 0>,<at 0.75 -1>,<at 1 0>]}}}
         Scale Curve=&RandomRangeFloat(Minimum=-1, Maximum=1)
        )
        B=&FloatFromCurve(FloatCurve=DataInterface() {propertyValues {Curve {keys [<at 0.95 0 RCIM_Cubic>, {arriveTangent=403.0874, interpMode=RCIM_Cubic, leaveTangent=403.0874, tangentMode=RCTM_User, time=0.9966, value=5}]}}})
       )
      )
      False Position:NiagaraPosition=&MakeVector(
       X=&FloatFromCurve(
        FloatCurve=DataInterface() {propertyValues {Curve {keys [<at 0 0>,<at 0.25 1>,<at 0.5 0>,<at 0.75 -1>,<at 1 0>]}}}
        Scale Curve=&RandomRangeFloat(Minimum=-1, Maximum=1)
       )
       Y=&Multiply_Float(
        A=&FloatFromCurve(
         FloatCurve=DataInterface() {propertyValues {Curve {keys [{arriveTangent=9.3254, interpMode=RCIM_Cubic, leaveTangent=9.3254, tangentMode=RCTM_User, time=0, value=0}, <at 0.2 1 RCIM_Cubic>]}}}
         Scale Curve=&RandomRangeFloat(Minimum=-2, Maximum=-4)
        )
        B=&FloatFromCurve(
         FloatCurve=DataInterface() {propertyValues {Curve {keys [<at 0 0>,<at 0.25 1>,<at 0.5 0>,<at 0.75 -1>,<at 1 0>]}}}
         Scale Curve=&RandomRangeFloat(Minimum=-0.5, Maximum=0.5)
        )
       )
       Z=&Add_Float(
        A=&FloatFromCurve(
         FloatCurve=DataInterface() {propertyValues {Curve {keys [{arriveTangent=9.3254, interpMode=RCIM_Cubic, leaveTangent=9.3254, tangentMode=RCTM_User, time=0, value=0}, <at 0.2 1 RCIM_Cubic>]}}}
         Scale Curve=&RandomRangeFloat(Minimum=-0.5, Maximum=-1.2)
        )
        B=&FloatFromCurve(FloatCurve=DataInterface() {propertyValues {Curve {keys [<at 0.95 0 RCIM_Cubic>, {arriveTangent=403.0874, interpMode=RCIM_Cubic, leaveTangent=403.0874, tangentMode=RCTM_User, time=0.9966, value=5}]}}})
       )
      )
     )
    )) {enabled=false, isSetParameters=true, module=SetVariables_#6}
   &CurlNoiseForce(
    Noise Frequency=75
    Noise Strength=&FloatFromCurve(
     FloatCurve=DataInterface() {propertyValues {Curve {keys [
         {arriveTangent=-39.1213, interpMode=RCIM_Cubic, leaveTangent=-39.1213, leaveTangentWeight=114.1625, tangentMode=RCTM_User, time=0, value=50}
         {arriveTangentWeight=0.3333, interpMode=RCIM_Cubic, leaveTangentWeight=53.6425, tangentMode=RCTM_User, time=1, value=25}
        ]}}}
     Scale Curve=&RandomRangeFloat(Minimum=2, Maximum=7)
    )
    Pan Noise=false
   )
   &Drag(
    Drag=&Multiply_Float(
     A=&RandomRangeFloat(Minimum=0.5, Maximum=1.5)
     B=&FloatFromCurve(FloatCurve=DataInterface() {propertyValues {Curve {defaultValue=1, keys [<at -0.0007 0>, {tangentMode=RCTM_User, time=0.1756, value=0.5471}, <at 1.0015 1.5>]}}})
    )
    Use Linear Drag=true
    Use Rotational Drag=true
   )
   &GravityForce(Gravity=&Multiply_Vector(
     A=&RandomRangeVector(
      Minimum=&VectorFromCurve(
       VectorCurve=DataInterface() {propertyValues {
         XCurve {keys [<at 0 0>,<at 1 0>]}
         YCurve {keys [<at 0 0>,<at 1 0>]}
         ZCurve {keys [
           {interpMode=RCIM_Cubic, tangentMode=RCTM_User, time=0, value=10}
           {arriveTangent=88.4209, interpMode=RCIM_Cubic, leaveTangent=88.4209, tangentMode=RCTM_User, time=0.75, value=35}
          ]}
        }}
       Scale Curve=<1 1 3>
      )
      Maximum=&VectorFromCurve(
       VectorCurve=DataInterface() {propertyValues {
         XCurve {keys [<at 0 0>,<at 1 0>]}
         YCurve {keys [<at 0 0>,<at 1 0>]}
         ZCurve {keys [
           {interpMode=RCIM_Cubic, tangentMode=RCTM_User, time=0, value=10}
           {arriveTangent=133.5137, interpMode=RCIM_Cubic, leaveTangent=133.5137, tangentMode=RCTM_User, time=0.75, value=70}
          ]}
        }}
       Scale Curve=<1 1 7>
      )
     )
     B=&MakeVector(
      X=0
      Y=0
      Z=&FloatFromCurve(
       FloatCurve=DataInterface() {propertyValues {Curve {defaultValue=1, keys [{tangentMode=RCTM_User, time=0.6542, value=0.5}, {tangentMode=RCTM_User, time=0.7019, value=1}]}}}
       Scale Curve=&RandomRangeFloat(Minimum=2, Maximum=3)
      )
     )
    ))
   &InheritVelocity(
    Inherited Velocity Amount Scale=&MakeCustomVectorFromBool(
     Bool=&SetBoolByFloatComparison(A=&VectorLength(Vector=$User.OwnerLocalVelocity), B=$Emitter.WalkForward Speed)
     True Vector=<0 0 0>
     False Vector=<-0.02 -0.02 0>
    )
    Speed Limit=true
    UseSpeedThreshold=false
   )
   &ScaleColor(
    Scale Alpha=&Multiply_Float(
     A=$User.FX_Alpha
     B=&FloatFromCurve(
      FloatCurve=DataInterface() {propertyValues {Curve {keys [
          {interpMode=RCIM_Cubic, leaveTangentWeight=0.3713, tangentMode=RCTM_User, time=0.5, value=1}
          {
           arriveTangent=-3.2389
           arriveTangentWeight=0.386
           interpMode=RCIM_Cubic
           leaveTangent=-3.2389
           leaveTangentWeight=0.1013
           tangentMode=RCTM_User
           time=0.8949
           value=0.5475
          }
          {arriveTangent=-1.4392, interpMode=RCIM_Cubic, leaveTangent=-1.4392, tangentMode=RCTM_User, time=1.0011, value=0.002}
         ]}}}
      Scale Curve=&Multiply_Float(A=1, B=$System.Bubbles TOD)
     )
    )
    ScaleA=true
    ScaleRGB=true
   ) {enabled=false}
   &SolveForcesAndVelocity(Clamp Velocity=false, Limit Acceleration=false)
   &BlendSpace("Position Blend (World <-> Local)":NiagaraFloat=&MakeCustomFloatFromBool(
     Bool=&SetBoolByFloatComparison(A=&VectorLength(Vector=$User.OwnerLocalVelocity), B=$Emitter.WalkForward Speed)
     True Float=1
     False Float=&Multiply_Float(
      A=1
      B=&FloatFromCurve(FloatCurve=DataInterface() {propertyValues {Curve {keys [<at 0.5 1>,<at 1 0.2>]}}})
     )
     Translation Contribution:NiagaraFloat=1
     Rotation Contribution:NiagaraFloat=1
    ))
   /SetVariables(
    Particles.Linear Color=&MultiplyLinearColor(
     Linear Color=&MultiplyLinearColor(
      Linear Color=$Particles.Color
      Scale Factor=&MakeVector4_FromLinearColor(LinearColor=&MakeLinearColorFromVectorAndFloat("Vector (RGB)"=<1 1 1>, "Float (Alpha)"=$System.Bubbles TOD))
     )
     Scale Factor=&MakeVector4_FromLinearColor(LinearColor=&MakeLinearColorFromVectorAndFloat(
       "Vector (RGB)"=<1 1 1>
       "Float (Alpha)"=&Multiply_Float(
        A=$User.FX_Alpha
        B=&FloatFromCurve(FloatCurve=DataInterface() {propertyValues {Curve {keys [
             <at 0.3605 0 RCIM_Cubic>
             <at 0.4703 1 RCIM_Cubic>
             {arriveTangent=-5.5857, interpMode=RCIM_Cubic, leaveTangent=-5.5857, tangentMode=RCTM_User, time=1, value=0}
            ]}}})
       )
      ))
    )
    Particles.Vector 2D:Vector2f=&Multiply_Vector2D_ByFloat(
     Float=&FloatFromCurve(FloatCurve=DataInterface() {propertyValues {Curve {keys [
          <at 0.3605 0 RCIM_Cubic>
          <at 0.4703 1 RCIM_Cubic>
          {arriveTangent=-1.2784, interpMode=RCIM_Cubic, leaveTangent=-1.2784, tangentMode=RCTM_User, time=1, value=0}
         ]}}})
     Vector2D=$Particles.SpriteSize
    )
   ) {isSetParameters=true}
  ]
  rendererClasses [&NiagaraSpriteRendererProperties]
  renderers [{rendererClass=&NiagaraSpriteRendererProperties}]
 }
 FXJelly_SpawnBubbles {
  emitterSpawnScript [/SetVariables(Emitter.IsTraversing Speed=100, Emitter.WalkForward Speed=300) {isSetParameters=true, module=SetVariables_#5}]
  emitterUpdateScript [
   &EmitterState(Life Cycle Mode=::NewEnumerator1, Loop Behavior=::NewEnumerator1, Loop Duration=2, UseLoopDelay=false)
   &SpawnParticlesFromOtherEmitter(
    Attribute Reader=DataInterface() {propertyValues {EmitterBinding {bindingMode=Other, emitterName=JellyFish}}}
    Spawn Rate=&FloatFromCurve(
     FloatCurve=DataInterface() {propertyValues {Curve {keys [
         {interpMode=RCIM_Cubic, leaveTangentWeight=0.0319, tangentMode=RCTM_User, time=0.0114, value=1}
         {arriveTangentWeight=0.6809, interpMode=RCIM_Cubic, leaveTangentWeight=0.0295, tangentMode=RCTM_User, time=0.0275, value=1}
         {interpMode=RCIM_Cubic, tangentMode=RCTM_User, time=0.0275, value=0}
         <at 0.0703 -0.001 RCIM_Cubic>
        ]}}}
     CurveIndex=$Emitter.NormalizedLoopAge
     Scale Curve=250
    )
    Spawn Rate Per Particle Cap=400
    Use Maximum Age=false
    Use Minimum Age=false
   )
  ]
  particleSpawnScript [
   &InitializeParticle(
    Lifetime Max=0.65
    Lifetime Min=0.2
    Lifetime Mode=::NewEnumerator1
    Mass Mode=::NewEnumerator3
    Sprite Rotation Mode=::NewEnumerator2
    Sprite Size Mode=::NewEnumerator1
    Uniform Sprite Size Max=12
    Uniform Sprite Size Min=3
    UsePositionOffset=false
    Write Material Random=true
   )
   &SampleParticlesFromOtherEmitter(Maximum Sampled Particle Age=0.2, Module.MaxSampledAgeCutoff=true, Module.MinSampledAgeCutoff=false, Reject Particle By Sampled Age=true)
   &ShapeLocation(MaskSpawnGroup=false, Offset=<0 0 -20>, Offset Mode=::NewEnumerator0, Override Seed=false, Sphere Radius=10, UseOwnerScale=true)
   &CameraOffset(Camera Offset Amount=&RandomRangeFloat(Minimum=3, Maximum=9))
  ]
  particleUpdateScript [
   &ParticleState()
   &ScaleSpriteSize(
    Non-Uniform Curve Scale=&RandomRangeVector2D(Minimum=<1 1>, Maximum=<1 1.3>)
    Non-Uniform Curve Sprite Scale=DataInterface() {propertyValues {
      XCurve {
       defaultValue=0
       keys [
        {arriveTangent=20.4729, interpMode=RCIM_Cubic, leaveTangent=20.4729, leaveTangentWeight=0.2846, tangentMode=RCTM_User, time=0, value=0}
        {
         arriveTangent=-9.5481
         arriveTangentWeight=0.0421
         interpMode=RCIM_Cubic
         leaveTangent=-0.4416
         leaveTangentWeight=0.2379
         tangentMode=RCTM_Break
         time=0.1263
         value=1
        }
        {arriveTangent=-1.1446, interpMode=RCIM_Cubic, leaveTangent=-1.1446, time=0.8401, value=0.6963}
        {arriveTangent=-10.9842, arriveTangentWeight=0.1, interpMode=RCIM_Cubic, leaveTangent=-10.9842, tangentMode=RCTM_User, time=1, value=0}
       ]
      }
      YCurve {keys [
        {arriveTangent=20.4729, interpMode=RCIM_Cubic, leaveTangent=20.4729, tangentMode=RCTM_User, time=0, value=0}
        {arriveTangent=-9.5481, interpMode=RCIM_Cubic, leaveTangent=-0.4416, tangentMode=RCTM_Break, time=0.1263, value=1}
        {arriveTangent=-1.1446, interpMode=RCIM_Cubic, leaveTangent=-1.1446, time=0.8401, value=0.6963}
        {arriveTangent=-10.9842, interpMode=RCIM_Cubic, leaveTangent=-10.9842, tangentMode=RCTM_User, time=1, value=0}
       ]}
     }}
    Scale Sprite Size Mode=::NewEnumerator3
   )
   &DynamicMaterialParameters(
    Index 0 Param 1=&FloatFromCurve(FloatCurve=DataInterface() {propertyValues {Curve {keys [
         {arriveTangent=0.1706, interpMode=RCIM_Cubic, leaveTangent=0.1706, leaveTangentWeight=0.3081, tangentMode=RCTM_User, time=0, value=0}
         {arriveTangentWeight=0.3104, interpMode=RCIM_Cubic, leaveTangentWeight=0.0345, tangentMode=RCTM_User, time=0.9, value=0.2}
         {arriveTangent=3.2005, interpMode=RCIM_Cubic, leaveTangent=3.2005, tangentMode=RCTM_User, time=1, value=0.5}
        ]}}})
    Module.Index0Float0Write=true
    Module.Index0Float1Write=true
    Module.Index0Float2Write=true
    Module.Index0Float3Write=true
   )
   &ScaleColor(
    Scale Alpha=&Multiply_Float(
     A=1
     B=&FloatFromCurve(
      FloatCurve=DataInterface() {propertyValues {Curve {keys [
          <at 0.1014 0 RCIM_Cubic>
          {interpMode=RCIM_Cubic, leaveTangentWeight=0.3713, tangentMode=RCTM_User, time=0.5, value=1}
          {
           arriveTangent=-3.2389
           arriveTangentWeight=0.386
           interpMode=RCIM_Cubic
           leaveTangent=-3.2389
           leaveTangentWeight=0.1013
           tangentMode=RCTM_User
           time=0.8949
           value=0.5475
          }
          {arriveTangent=-1.4392, interpMode=RCIM_Cubic, leaveTangent=-1.4392, tangentMode=RCTM_User, time=1.0011, value=0.002}
         ]}}}
      Scale Curve=&Multiply_Float(A=1, B=$System.Bubbles TOD)
     )
    )
    ScaleA=true
    ScaleRGB=true
   )
   /SetVariables(Particles.Position=&AddVectorToPosition(
     Position=$Particles.Position
     Vector=&MakeVector(
      X=&FloatFromCurve(
       FloatCurve=DataInterface() {propertyValues {Curve {keys [<at 0 0>,<at 0.25 1>,<at 0.5 0>,<at 0.75 -1>,<at 1 0>]}}}
       Scale Curve=&RandomRangeFloat(Minimum=-1, Maximum=1)
      )
      Y=&FloatFromCurve(
       FloatCurve=DataInterface() {propertyValues {Curve {keys [<at 0 0>,<at 0.25 1>,<at 0.5 0>,<at 0.75 -1>,<at 1 0>]}}}
       Scale Curve=&RandomRangeFloat(Minimum=-0.5, Maximum=0.5)
      )
      Z=&Add_Float(
       A=&FloatFromCurve(
        FloatCurve=DataInterface() {propertyValues {Curve {keys [{arriveTangent=-1.0003, interpMode=RCIM_Cubic, leaveTangent=-1.0003, tangentMode=RCTM_User, time=0, value=0}, <at 0.2 1 RCIM_Cubic>]}}}
        Scale Curve=&RandomRangeFloat(Minimum=0.001, Maximum=0.002)
       )
       B=&FloatFromCurve(
        FloatCurve=DataInterface() {propertyValues {Curve {keys [<at 0.95 0 RCIM_Cubic>, {arriveTangent=403.0874, interpMode=RCIM_Cubic, leaveTangent=403.0874, tangentMode=RCTM_User, time=0.9966, value=5}]}}}
        Scale Curve=0.001
       )
      )
     )
    )) {isSetParameters=true, module=SetVariables_#6}
   &CurlNoiseForce(
    Noise Frequency=75
    Noise Strength=&FloatFromCurve(
     FloatCurve=DataInterface() {propertyValues {Curve {keys [
         {arriveTangent=-39.1213, interpMode=RCIM_Cubic, leaveTangent=-39.1213, leaveTangentWeight=114.1625, tangentMode=RCTM_User, time=0, value=50}
         {arriveTangentWeight=0.3333, interpMode=RCIM_Cubic, leaveTangentWeight=53.6425, tangentMode=RCTM_User, time=1, value=25}
        ]}}}
     Scale Curve=&RandomRangeFloat(Minimum=5, Maximum=1)
    )
    Pan Noise=false
   )
   &GravityForce(Gravity=<0 0 200>)
   &SolveForcesAndVelocity(Clamp Velocity=false, Limit Acceleration=false)
   &BlendSpace("Position Blend (World <-> Local)":NiagaraFloat=&MakeCustomFloatFromBool(
     Bool=&SetBoolByFloatComparison(A=&VectorLength(Vector=$User.OwnerLocalVelocity), B=$Emitter.WalkForward Speed)
     True Float=1
     False Float=&Multiply_Float(
      A=1
      B=&FloatFromCurve(FloatCurve=DataInterface() {propertyValues {Curve {keys [<at 0.5 1>,<at 1 0.2>]}}})
     )
     Translation Contribution:NiagaraFloat=1
     Rotation Contribution:NiagaraFloat=1
    ))
   &self:WriteDeath_01(Position Array:NiagaraDataInterfaceArrayPosition=$System.DeathArray) {module=WriteDeath_01}
  ]
  rendererClasses [&NiagaraSpriteRendererProperties]
  renderers [{rendererClass=&NiagaraSpriteRendererProperties}]
 }
 JellyFish {
  emitterSpawnScript [/SetVariables(Emitter.IsTraversing Speed=100, Emitter.WalkForward Speed=300) {isSetParameters=true, module=SetVariables_#7}]
  emitterUpdateScript [
   &EmitterState(Life Cycle Mode=::NewEnumerator1, Loop Behavior=::NewEnumerator1, Loop Duration Mode=::NewEnumerator1, UseLoopDelay=false)
   &SpawnBurst_Instantaneous(Spawn Count=5, Use Spawn Probability=false, UseLoopCountLimit=false)
   &SpawnBurst_Instantaneous(Spawn Count=5, Spawn Time=0.5833, Use Spawn Probability=false, UseLoopCountLimit=false) {enabled=false, module=SpawnBurst_Instantaneous001}
  ]
  particleSpawnScript [
   /SetVariables(Particles.VisibilityTag=&SpawnGroup()) {isSetParameters=true, module=SetVariables_#8}
   &PartitionParticles(Number Of Partitions=5, Particles Per Partition=1, Partition Mode:ENiagara_PartitionMode_v2=::NewEnumerator1)
   &InitializeParticle(
    Color Mode=::NewEnumerator0
    Lifetime=0
    Mass Mode=::NewEnumerator3
    Position Offset=<0 0 75>
    Sprite Size Mode=::NewEnumerator3
    Uniform Sprite Size=6
    UsePositionOffset=true
    Write Material Random=true
   )
   &ShapeLocation(
    MaskSpawnGroup=false
    Non Uniform Scale=<0.3 1 1>
    Offset=&MergeHelper_VectorSwitch(
     0=<-60 -10 -30>
     1=<-20 80 30>
     2=<50 15 70>
     3=<-50 -100 20>
     4=<95 -15 20>
     5=<-120 -50 30>
     Index=$Particles.PartitionParticles.Partition
     Num=::NewEnumerator4
    )
    Offset Mode=::NewEnumerator0
    Override Seed=false
    Sphere Radius=10
    UseOwnerScale=true
   )
   &CameraOffset(Camera Offset Amount=&RandomRangeFloat(Minimum=-2, Maximum=-6))
  ]
  particleUpdateScript [
   &ParticleState(Kill Particles When Lifetime Has Elapsed=false, Let Infinitely Lived Particles Die When Emitter Deactivates=true)
   /SetVariables(Particles.Debug Partition Color:LinearColor=&MergeHelper_ColorSwitch(
     0=<rgba 1 0.0746 0.014 1>
     1=<rgba 0 0.2747 1 1>
     2=<rgba 1 0.553 0 1>
     3=<rgba 0.7829 0 1 1>
     4=<rgba 0 1 0.1336 1>
     5=<rgba 0.2501 0 1 1>
     Index=$Particles.PartitionParticles.Partition
     Num=::NewEnumerator4
    )) {isSetParameters=true, module=SetVariables_#9}
   /SetVariables(Particles.Bobbing Sine:NiagaraFloat=&Waveform(
     "Waveform [1]"=::NewEnumerator0
     "[1] Frequency"=0.5
     "[1] Amplitude Min/Max"=&Multiply_Vector2D_ByFloat(Float=&RandomRangeFloat(Minimum=0.3, Maximum=0.5), Vector2D=<-1 1>)
    )) {isSetParameters=true, module=SetVariables_#10}
   &Timeline(
    Play=&SetBoolByFloatComparison(A=&VectorLength(Vector=$User.OwnerLocalVelocity), B=$Emitter.IsTraversing Speed)
    Play Rate=&RandomRangeFloat(Minimum=0.25, Maximum=5)
    Rewind Rate:NiagaraFloat=&RandomRangeFloat(Minimum=0.25, Maximum=5)
   )
   &SubUVAnimation(
    End Frame Range Override=10
    Play Rate:NiagaraFloat=&MergeHelper_FloatSwitch(0=0.9, 1=0.8, 2=0.7, 3=0.85, 4=0.75, 5=0.7, Index=$Particles.PartitionParticles.Partition, Num=::NewEnumerator4)
    Playback Scale Curve=DataInterface() {propertyValues {Curve {keys [
        {interpMode=RCIM_Cubic, tangentMode=RCTM_User, time=0.25, value=1}
        {arriveTangent=-1.2656, interpMode=RCIM_Cubic, leaveTangent=-1.2656, tangentMode=RCTM_User, time=1, value=0}
       ]}}}
    Playback Scale Curve Index:NiagaraFloat=&Frac_Float(Float=$Particles.NormalizedAge)
    "Sprite Renderer "=DataInterface() {propertyValues {SpriteRenderer=&self:JellyFish_0.NiagaraSpriteRendererProperties_1}}
    SubUV Animation Mode:ENiagara_SubUVLookupModeV2=::NewEnumerator3
    UseEndFrame=true
    UsePlayRateScale=true
    UseStartFrame=true
   ) {module=SubUVAnimation001}
   &ScaleSpriteSize(
    Scale Sprite Size Mode=::NewEnumerator0
    Uniform Scale Factor=&Waveform(
     "Waveform [1]"=::NewEnumerator0
     "[1] Frequency"=&MergeHelper_FloatSwitch(0=0.8, 1=0.6, 2=0.7, 3=0.65, 4=0.75, 5=0.7, Index=$Particles.PartitionParticles.Partition, Num=::NewEnumerator4)
     "[1] Amplitude Min/Max"=<0 1>
    )
   ) {enabled=false}
   &ScaleSpriteSize(
    Scale Sprite Size Mode=::NewEnumerator0
    Uniform Scale Factor=&MakeCustomFloatFromBool(
     Bool=&Boolean_OrOperation(
      Bool1=&SetBoolByFloatComparison(A=$Particles.SubImageIndex, B=8)
      Bool2=&SetBoolByFloatComparison(A=$Particles.SubImageIndex, B=3, Comparison Type=::NewEnumerator6)
     )
     True Float=&RandomRangeFloat(Minimum=0.7, Maximum=1)
     False Float=0
    )
   ) {module=ScaleSpriteSize001}
   /SetVariables() {isSetParameters=true}
   &ScaleColor(
    Scale Alpha=&Waveform(
     "Waveform [1]"=::NewEnumerator0
     "[1] Frequency"=&MergeHelper_FloatSwitch(0=0.8, 1=0.6, 2=0.7, 3=0.65, 4=0.75, 5=0.7, Index=$Particles.PartitionParticles.Partition, Num=::NewEnumerator4)
     "[1] Amplitude Min/Max"=<0 1>
    )
    Scale RGB=&MakeVectorFromLinearColorRGB(Color=<rgba 2 2 2 1>)
    ScaleA=true
    ScaleRGB=true
   ) {enabled=false}
   &ScaleColor(
    Scale Alpha=&MakeCustomFloatFromBool(
     Bool=&Boolean_OrOperation(
      Bool1=&SetBoolByFloatComparison(A=$Particles.SubImageIndex, B=8)
      Bool2=&SetBoolByFloatComparison(A=$Particles.SubImageIndex, B=3, Comparison Type=::NewEnumerator6)
     )
     True Float=1
     False Float=0
    )
    ScaleA=true
    ScaleRGB=true
   ) {module=ScaleColor001}
   /SetVariables(Particles.Primary Color:LinearColor=&FNTimeOfDayColor(
     Night Color=<rgba 0.5525 0.4948 1 1>
     Morning Color=<rgba 1.185 1.2 0.5082 1>
     Day Color=<rgba 1 1 1 1>
     Evening Color=<rgba 1.185 1.2 0.5082 1>
    )) {isSetParameters=true, module=SetVariables_#11}
   /SetVariables(Particles.Position=&AddVectorToPosition(
     Position=$Particles.Position
     Vector=&Add_Vector(
      A=&Multiply_VectorByFloat(
       Vector=&MergeHelper_VectorSwitch(
        0=<-0.5 0 0>
        1=<0.5 0 0>
        2=<0.4 0 0>
        3=<0.5 0 0>
        4=<0.4 0 0>
        5=<0.15 0 0>
        Index=$Particles.Initial.PartitionParticles.Partition
        Num=::NewEnumerator4
       )
       Float=&Waveform(
        "Waveform [1]"=::NewEnumerator0
        "[1] Frequency"=&RandomRangeFloat(Minimum=0.08, Maximum=0.12)
        "[1] Amplitude Min/Max"=&Multiply_Vector2D_ByFloat(Float=&RandomRangeFloat(Minimum=0.5, Maximum=0.7), Vector2D=<-1 1>)
       )
      )
      B=&Multiply_VectorByFloat(
       Vector=&MergeHelper_VectorSwitch(
        0=&RandomRangeVector(Minimum=<0 1.2 1>, Maximum=<0 1.4 0.8>)
        1=&RandomRangeVector(Minimum=<0 -0.5 -0.5>, Maximum=<0 -0.3 -0.7>)
        2=&RandomRangeVector(Minimum=<0 -0.4 -0.6>, Maximum=<0 -0.5 -0.5>)
        3=&RandomRangeVector(Minimum=<0 0.7 0.4>, Maximum=<0 1 0.6>)
        4=&RandomRangeVector(Minimum=<0 0.4 0.6>, Maximum=<0 0.7 0.75>)
        5=&RandomRangeVector(Minimum=<0 -1 -1>, Maximum=<0 1.25 1>, Evaluation Type=::NewEnumerator1)
        Index=$Particles.PartitionParticles.Partition
        Num=::NewEnumerator4
       )
       Float=&Waveform(
        "Waveform [1]"=::NewEnumerator0
        "[1] Frequency"=&RandomRangeFloat(Minimum=0.05, Maximum=0.13)
        "[1] Amplitude Min/Max"=&Multiply_Vector2D_ByFloat(Float=&RandomRangeFloat(Minimum=0.5, Maximum=0.7), Vector2D=<-1 1>)
       )
      )
     )
    )) {isSetParameters=true, module=SetVariables_#12}
   /SetVariables(Particles.PivotOffset:Vector2f=&MakeVector2D(
     X=0.5
     Y=&Waveform(
      "Waveform [1]"=::NewEnumerator0
      "[1] Frequency"=&MergeHelper_FloatSwitch(0=0.8, 1=0.6, 2=0.7, 3=0.65, 4=0.75, 5=0.7, Index=$Particles.PartitionParticles.Partition, Num=::NewEnumerator4)
      "[1] Amplitude Min/Max"=<0.25 -0.25>
     )
    )) {isSetParameters=true, module=SetVariables_#13}
   &GravityForce(Gravity=&MaskVector(
     Mask Source=0
     Mask Parameter=?
     Mask=0
     Masked Vector=<0 0 0>
     Passthrough Vector=&MakeVector(X=0, Y=-50, Z=-50)
    ))
   /SetVariables(Particles.Bubbles Dynamic Param:Vector4f=&MakeVector4(
     X=&FloatFromCurve(FloatCurve=DataInterface() {propertyValues {Curve {keys [
          {arriveTangent=0.1706, interpMode=RCIM_Cubic, leaveTangent=0.1706, leaveTangentWeight=0.3081, tangentMode=RCTM_User, time=0, value=0}
          {arriveTangentWeight=0.3104, interpMode=RCIM_Cubic, leaveTangentWeight=0.0345, tangentMode=RCTM_User, time=0.9, value=0.2}
          {arriveTangent=3.2005, interpMode=RCIM_Cubic, leaveTangent=3.2005, tangentMode=RCTM_User, time=1, value=0.5}
         ]}}})
     Y=0
     Z=0
     W=0
    )) {isSetParameters=true, module=SetVariables_#14}
   &PointForce(
    Force Origin=&AddVectorToPosition(Position=/Niagara/DynamicInputs/Helpers/SimulationPosition(), Vector=<0 0 100>)
    Force Strength=&FloatFromCurve(FloatCurve=DataInterface() {propertyValues {Curve {keys [<at 0 0>,<at 0.1 1>]}}}, CurveIndex=$User.Jelly_Push, Scale Curve=500)
    Module.Use Falloff=false
    Module.UseLinearForce=false
   )
   &SolveForcesAndVelocity(Clamp Velocity=false, Limit Acceleration=false)
   /SetVariables(Particles.Vector 2D:Vector2f=&Multiply_Vector2D(
     A=&Vector2DFromFloat(Value=&MergeHelper_FloatSwitch(0=28.5, 1=25, 2=28, 3=25, 4=27, 5=25, Index=$Particles.PartitionParticles.Partition, Num=::NewEnumerator4))
     B=&Vector2DFromFloat(Value=&FloatFromCurve(
       FloatCurve=DataInterface() {propertyValues {Curve {keys [<at 0.5 1>,<at 1 0>]}}}
       CurveIndex=$User.Jelly_Push
       Scale Curve=&FloatFromCurve(FloatCurve=DataInterface() {propertyValues {Curve {keys [<at 0 1>]}}}, CurveIndex=$Particles.Age, Scale Curve=1.1)
      ))
    )) {isSetParameters=true, module=SetVariables_#15}
  ]
  rendererClasses [&NiagaraSpriteRendererProperties]
  renderers [{rendererClass=&NiagaraSpriteRendererProperties}]
 }
 JellyFish001 {
  bEnabled=false
  emitterSpawnScript [/SetVariables(Emitter.IsTraversing Speed=100, Emitter.WalkForward Speed=300) {isSetParameters=true, module=SetVariables_#7}]
  emitterUpdateScript [
   &EmitterState(Life Cycle Mode=::NewEnumerator1, Loop Behavior=::NewEnumerator1, Loop Duration Mode=::NewEnumerator1, UseLoopDelay=false)
   &SpawnBurst_Instantaneous(Spawn Count=5, Use Spawn Probability=false, UseLoopCountLimit=false)
  ]
  particleSpawnScript [
   /SetVariables(Particles.VisibilityTag=&SpawnGroup()) {isSetParameters=true, module=SetVariables_#8}
   &PartitionParticles(Number Of Partitions=5, Particles Per Partition=1, Partition Mode:ENiagara_PartitionMode_v2=::NewEnumerator1)
   &InitializeParticle(
    Color Mode=::NewEnumerator0
    Lifetime=$Emitter.CurrentLoopDuration
    Mass Mode=::NewEnumerator3
    Position Offset=<0 0 75>
    Sprite Size Mode=::NewEnumerator3
    Uniform Sprite Size=6
    UsePositionOffset=true
    Write Material Random=true
   )
   &ShapeLocation(
    MaskSpawnGroup=false
    Non Uniform Scale=<0.3 1 1>
    Offset=&MergeHelper_VectorSwitch(
     0=<-77 -30 -10>
     1=<-130 30 60>
     2=<70 -20 90>
     3=<80 50 20>
     4=<95 -15 -10>
     5=<-120 -50 30>
     Index=$Particles.PartitionParticles.Partition
     Num=::NewEnumerator4
    )
    Offset Mode=::NewEnumerator0
    Override Seed=false
    Sphere Radius=10
    UseOwnerScale=true
   )
   &CameraOffset(Camera Offset Amount=&RandomRangeFloat(Minimum=-2, Maximum=-6))
  ]
  particleUpdateScript [
   &ParticleState()
   &ScaleSpriteSize(
    Scale Sprite Size Mode=::NewEnumerator0
    Uniform Scale Factor=&Waveform(
     "Waveform [1]"=::NewEnumerator0
     "[1] Frequency"=&MergeHelper_FloatSwitch(0=0.8, 1=0.6, 2=0.7, 3=0.65, 4=0.75, 5=0.7, Index=$Particles.PartitionParticles.Partition, Num=::NewEnumerator4)
     "[1] Amplitude Min/Max"=<0 1>
    )
   )
   /SetVariables(Particles.Debug Partition Color:LinearColor=&MergeHelper_ColorSwitch(
     0=<rgba 1 0.0746 0.014 1>
     1=<rgba 0 0.2747 1 1>
     2=<rgba 1 0.553 0 1>
     3=<rgba 0.7829 0 1 1>
     4=<rgba 0 1 0.1336 1>
     5=<rgba 0.2501 0 1 1>
     Index=$Particles.PartitionParticles.Partition
     Num=::NewEnumerator4
    )) {isSetParameters=true, module=SetVariables_#9}
   /SetVariables(Particles.Bobbing Sine:NiagaraFloat=&Waveform(
     "Waveform [1]"=::NewEnumerator0
     "[1] Frequency"=0.5
     "[1] Amplitude Min/Max"=&Multiply_Vector2D_ByFloat(Float=&RandomRangeFloat(Minimum=0.3, Maximum=0.5), Vector2D=<-1 1>)
    )) {isSetParameters=true, module=SetVariables_#10}
   &Timeline(
    Play=&SetBoolByFloatComparison(A=&VectorLength(Vector=$User.OwnerLocalVelocity), B=$Emitter.IsTraversing Speed)
    Play Rate=&RandomRangeFloat(Minimum=0.25, Maximum=5)
    Rewind Rate:NiagaraFloat=&RandomRangeFloat(Minimum=0.25, Maximum=5)
   )
   &SubUVAnimation(
    End Frame Range Override=10
    Play Rate:NiagaraFloat=&MergeHelper_FloatSwitch(0=0.8, 1=0.6, 2=0.7, 3=0.65, 4=0.75, 5=0.7, Index=$Particles.PartitionParticles.Partition, Num=::NewEnumerator4)
    Playback Scale Curve=DataInterface() {propertyValues {Curve {keys [
        {interpMode=RCIM_Cubic, tangentMode=RCTM_User, time=0.25, value=1}
        {arriveTangent=-1.2656, interpMode=RCIM_Cubic, leaveTangent=-1.2656, tangentMode=RCTM_User, time=1, value=0}
       ]}}}
    Playback Scale Curve Index:NiagaraFloat=$Particles.Bobbing Sine
    "Sprite Renderer "=DataInterface() {propertyValues {SpriteRenderer=&self:JellyFish001_0.NiagaraSpriteRendererProperties_1}}
    SubUV Animation Mode:ENiagara_SubUVLookupModeV2=::NewEnumerator3
    UseEndFrame=true
    UsePlayRateScale=true
    UseStartFrame=true
   ) {module=SubUVAnimation001}
   &SpriteFacingAndAlignment(
    Sprite Alignment=&Lerp_Vector(
     A=&RandomRangeVector(Minimum=<0 0 1>, Maximum=<0 0.4 1>)
     B=<0 1 0.25>
     Alpha=&RemapRange(
      Input Value To Remap=&VectorLength(Vector=$User.OwnerLocalVelocity)
      Input Min=20
      Input Max=100
      Output Min=0
      Output Max=1
     )
    )
    WriteAlignment=true
    WriteFacing=false
   )
   &ScaleColor(
    Scale Alpha=&Waveform(
     "Waveform [1]"=::NewEnumerator0
     "[1] Frequency"=&MergeHelper_FloatSwitch(0=0.8, 1=0.6, 2=0.7, 3=0.65, 4=0.75, 5=0.7, Index=$Particles.PartitionParticles.Partition, Num=::NewEnumerator4)
     "[1] Amplitude Min/Max"=<0 1>
    )
    Scale RGB=&MakeVectorFromLinearColorRGB(Color=<rgba 2 2 2 1>)
    ScaleA=true
    ScaleRGB=true
   )
   /SetVariables(Particles.Primary Color:LinearColor=&MakeLinearColorFromVectorAndFloat(
     "Vector (RGB)"=&MakeVectorFromLinearColorRGB(Color=&MultiplyLinearColor(
       Linear Color=&FNTimeOfDayColor(
        Night Color=<rgba 0.4061 0.3211 0.6698 1>
        Morning Color=<rgba 1 0.8846 0.7454 1>
        Day Color=<rgba 1 0.7379 0.7437 1>
        Evening Color=<rgba 1 0.7177 0.7864 1>
       )
       Scale Factor=&MakeVector4_FromLinearColor(LinearColor=<rgba 1 1 1 1>)
      ))
     "Float (Alpha)"=&Multiply_Float(
      A=&Multiply_Float(A=1, B=$User.FX_Alpha)
      B=&FloatFromCurve(
       FloatCurve=DataInterface() {propertyValues {Curve {keys [<at 0.4 1>,<at 0.8 0>]}}}
       CurveIndex=&Abs_Float(Float:NiagaraFloat=&DotProduct(
         Vector A=$Particles.SpriteAlignment
         Vector B=&ReturnCameraVector(Camera Vectors=::NewEnumerator1, Coordinate Space=::Simulation)
        ))
      )
     )
    )) {isSetParameters=true, module=SetVariables_#11}
   /SetVariables(
    Particles.Secondary Color:LinearColor=&MakeLinearColorFromVectorAndFloat(
     "Vector (RGB)"=&MakeVectorFromLinearColorRGB(Color=&MultiplyLinearColor(
       Linear Color=&FNTimeOfDayColor(
        Night Color=<rgba 0.4061 0.3211 0.6698 1>
        Morning Color=<rgba 1 0.8846 0.7454 1>
        Day Color=<rgba 1 0.7379 0.7437 1>
        Evening Color=<rgba 1 0.7177 0.7864 1>
       )
       Scale Factor=&MakeVector4_FromLinearColor(LinearColor=<rgba 1 1 1 1>)
      ))
     "Float (Alpha)"=&Multiply_Float(
      A=&Multiply_Float(A=1, B=$User.FX_Alpha)
      B=&FloatFromCurve(
       FloatCurve=DataInterface() {propertyValues {Curve {keys [<at 0.5 0>,<at 1 1>]}}}
       CurveIndex=&Abs_Float(Float:NiagaraFloat=&DotProduct(
         Vector A=$Particles.SpriteAlignment
         Vector B=&ReturnCameraVector(Camera Vectors=::NewEnumerator1, Coordinate Space=::Simulation)
        ))
      )
     )
    )
    Particles.Secondary SpriteAlignment:Vector3f=<0 1 1>
   ) {isSetParameters=true, module=SetVariables_#16}
   /SetVariables(Particles.Position=&AddVectorToPosition(
     Position=$Particles.Position
     Vector=&Multiply_VectorByFloat(
      Vector=&MergeHelper_VectorSwitch(
       0=&RandomRangeVector(Minimum=<-0.4 1.5 1>, Maximum=<-0.1 1.7 1.25>)
       1=&RandomRangeVector(Minimum=<-0.25 -0.5 -0.5>, Maximum=<0.25 -0.3 -0.7>)
       2=&RandomRangeVector(Minimum=<-0.25 -0.4 -0.6>, Maximum=<0.25 -0.5 -0.5>)
       3=&RandomRangeVector(Minimum=<-0.25 0.7 0.4>, Maximum=<0.25 1 0.6>)
       4=&RandomRangeVector(Minimum=<-0.25 0.4 0.6>, Maximum=<0.25 0.7 0.75>)
       5=&RandomRangeVector(Minimum=<-0.25 -1 -1>, Maximum=<0.25 1.25 1>, Evaluation Type=::NewEnumerator1)
       Index=$Particles.PartitionParticles.Partition
       Num=::NewEnumerator4
      )
      Float=&Waveform(
       "Waveform [1]"=::NewEnumerator0
       "[1] Frequency"=&RandomRangeFloat(Minimum=0.05, Maximum=0.13)
       "[1] Amplitude Min/Max"=&Multiply_Vector2D_ByFloat(Float=&RandomRangeFloat(Minimum=0.5, Maximum=0.7), Vector2D=<-1 1>)
      )
     )
    )) {isSetParameters=true, module=SetVariables_#17}
   /SetVariables(Particles.PivotOffset:Vector2f=&MakeVector2D(
     X=0.5
     Y=&Waveform(
      "Waveform [1]"=::NewEnumerator0
      "[1] Frequency"=&MergeHelper_FloatSwitch(0=0.8, 1=0.6, 2=0.7, 3=0.65, 4=0.75, 5=0.7, Index=$Particles.PartitionParticles.Partition, Num=::NewEnumerator4)
      "[1] Amplitude Min/Max"=<0.25 -0.25>
     )
    )) {isSetParameters=true, module=SetVariables_#13}
   &GravityForce(Gravity=&MaskVector(
     Mask Source=0
     Mask Parameter=?
     Mask=0
     Masked Vector=<0 0 0>
     Passthrough Vector=&MakeVector(X=0, Y=-50, Z=-50)
    ))
   /SetVariables(Particles.Bubbles Dynamic Param:Vector4f=&MakeVector4(
     X=&FloatFromCurve(FloatCurve=DataInterface() {propertyValues {Curve {keys [
          {arriveTangent=0.1706, interpMode=RCIM_Cubic, leaveTangent=0.1706, leaveTangentWeight=0.3081, tangentMode=RCTM_User, time=0, value=0}
          {arriveTangentWeight=0.3104, interpMode=RCIM_Cubic, leaveTangentWeight=0.0345, tangentMode=RCTM_User, time=0.9, value=0.2}
          {arriveTangent=3.2005, interpMode=RCIM_Cubic, leaveTangent=3.2005, tangentMode=RCTM_User, time=1, value=0.5}
         ]}}})
     Y=0
     Z=0
     W=0
    )) {isSetParameters=true, module=SetVariables_#14}
   &SolveForcesAndVelocity(Clamp Velocity=false, Limit Acceleration=false)
   /SetVariables(
    Particles.Bubbles Color:LinearColor=&MakeLinearColorFromVectorAndFloat(
     "Vector (RGB)"=<1 1 1>
     "Float (Alpha)"=&Waveform(
      "Waveform [1]"=::NewEnumerator0
      "[1] Frequency"=&MergeHelper_FloatSwitch(0=0.8, 1=0.6, 2=0.7, 3=0.65, 4=0.75, 5=0.7, Index=$Particles.PartitionParticles.Partition, Num=::NewEnumerator4)
      "[1] Amplitude Min/Max"=<0 1>
     )
    )
    Particles.Vector 2D:Vector2f=&Vector2DFromFloat(Value=&Waveform(
      "Waveform [1]"=::NewEnumerator0
      "[1] Frequency"=&MergeHelper_FloatSwitch(0=0.8, 1=0.6, 2=0.7, 3=0.65, 4=0.75, 5=0.7, Index=$Particles.PartitionParticles.Partition, Num=::NewEnumerator4)
      "[1] Amplitude Min/Max"=<0.27 -0.27>
     ))
   ) {isSetParameters=true}
   &DynamicMaterialParameters(Module.Index0Float0Write=true, Module.Index0Float1Write=true, Module.Index0Float2Write=true, Module.Index0Float3Write=true) {enabled=false}
   /SetVariables(Particles.Vector 2D:Vector2f=&Vector2DFromFloat(Value=&MaskFloat(
      Mask Source=0
      Mask Parameter=?
      Mask=0
      Masked Float=&MergeHelper_FloatSwitch(0=55, 1=25, 2=33, 3=25, 4=45, 5=25, Index=$Particles.PartitionParticles.Partition, Num=::NewEnumerator4)
      Passthrough Float=&RandomRangeFloat(Minimum=1.5, Maximum=6)
     ))) {isSetParameters=true, module=SetVariables_#15}
  ]
  rendererClasses [&NiagaraSpriteRendererProperties]
  renderers [
   {rendererClass=&NiagaraSpriteRendererProperties}
   {rendererClass=&NiagaraSpriteRendererProperties, rendererIndex=1}
  ]
 }
 JellyFishBackup {
  bEnabled=false
  emitterSpawnScript [/SetVariables(Emitter.IsTraversing Speed=100, Emitter.WalkForward Speed=300) {isSetParameters=true, module=SetVariables_#7}]
  emitterUpdateScript [
   &EmitterState(Life Cycle Mode=::NewEnumerator1, Loop Behavior=::NewEnumerator1, Loop Duration Mode=::NewEnumerator1, UseLoopDelay=false)
   &SpawnBurst_Instantaneous(Spawn Count=5, Use Spawn Probability=false, UseLoopCountLimit=false)
   &SpawnBurst_Instantaneous(Spawn Count=5, Spawn Time=0.5833, Use Spawn Probability=false, UseLoopCountLimit=false) {enabled=false, module=SpawnBurst_Instantaneous001}
  ]
  particleSpawnScript [
   /SetVariables(Particles.VisibilityTag=&SpawnGroup()) {isSetParameters=true, module=SetVariables_#8}
   &PartitionParticles(Number Of Partitions=5, Particles Per Partition=1, Partition Mode:ENiagara_PartitionMode_v2=::NewEnumerator1)
   &InitializeParticle(
    Color Mode=::NewEnumerator0
    Lifetime=$Emitter.CurrentLoopDuration
    Mass Mode=::NewEnumerator3
    Position Offset=<0 0 75>
    Sprite Size Mode=::NewEnumerator3
    Uniform Sprite Size=6
    UsePositionOffset=true
    Write Material Random=true
   )
   &ShapeLocation(
    MaskSpawnGroup=false
    Non Uniform Scale=<0.3 1 1>
    Offset=&MergeHelper_VectorSwitch(
     0=<-60 -10 -30>
     1=<-20 80 30>
     2=<50 15 70>
     3=<-50 -100 20>
     4=<95 -15 20>
     5=<-120 -50 30>
     Index=$Particles.PartitionParticles.Partition
     Num=::NewEnumerator4
    )
    Offset Mode=::NewEnumerator0
    Override Seed=false
    Sphere Radius=10
    UseOwnerScale=true
   )
   &CameraOffset(Camera Offset Amount=&RandomRangeFloat(Minimum=-2, Maximum=-6))
  ]
  particleUpdateScript [
   &ParticleState()
   &ScaleSpriteSize(
    Scale Sprite Size Mode=::NewEnumerator0
    Uniform Scale Factor=&Waveform(
     "Waveform [1]"=::NewEnumerator0
     "[1] Frequency"=&MergeHelper_FloatSwitch(0=0.8, 1=0.6, 2=0.7, 3=0.65, 4=0.75, 5=0.7, Index=$Particles.PartitionParticles.Partition, Num=::NewEnumerator4)
     "[1] Amplitude Min/Max"=<0 1>
    )
   )
   /SetVariables(Particles.Debug Partition Color:LinearColor=&MergeHelper_ColorSwitch(
     0=<rgba 1 0.0746 0.014 1>
     1=<rgba 0 0.2747 1 1>
     2=<rgba 1 0.553 0 1>
     3=<rgba 0.7829 0 1 1>
     4=<rgba 0 1 0.1336 1>
     5=<rgba 0.2501 0 1 1>
     Index=$Particles.PartitionParticles.Partition
     Num=::NewEnumerator4
    )) {isSetParameters=true, module=SetVariables_#9}
   /SetVariables(Particles.Bobbing Sine:NiagaraFloat=&Waveform(
     "Waveform [1]"=::NewEnumerator0
     "[1] Frequency"=0.5
     "[1] Amplitude Min/Max"=&Multiply_Vector2D_ByFloat(Float=&RandomRangeFloat(Minimum=0.3, Maximum=0.5), Vector2D=<-1 1>)
    )) {isSetParameters=true, module=SetVariables_#10}
   &Timeline(
    Play=&SetBoolByFloatComparison(A=&VectorLength(Vector=$User.OwnerLocalVelocity), B=$Emitter.IsTraversing Speed)
    Play Rate=&RandomRangeFloat(Minimum=0.25, Maximum=5)
    Rewind Rate:NiagaraFloat=&RandomRangeFloat(Minimum=0.25, Maximum=5)
   )
   &SubUVAnimation(
    End Frame Range Override=10
    Play Rate:NiagaraFloat=&MergeHelper_FloatSwitch(0=0.8, 1=0.6, 2=0.7, 3=0.65, 4=0.75, 5=0.7, Index=$Particles.PartitionParticles.Partition, Num=::NewEnumerator4)
    Playback Scale Curve=DataInterface() {propertyValues {Curve {keys [
        {interpMode=RCIM_Cubic, tangentMode=RCTM_User, time=0.25, value=1}
        {arriveTangent=-1.2656, interpMode=RCIM_Cubic, leaveTangent=-1.2656, tangentMode=RCTM_User, time=1, value=0}
       ]}}}
    "Sprite Renderer "=DataInterface() {propertyValues {SpriteRenderer=&self:JellyFishBackup_0.NiagaraSpriteRendererProperties_1}}
    SubUV Animation Mode:ENiagara_SubUVLookupModeV2=::NewEnumerator3
    UseEndFrame=true
    UsePlayRateScale=true
    UseStartFrame=true
   ) {module=SubUVAnimation001}
   &SpriteFacingAndAlignment(Sprite Alignment=<0 0 1>, WriteAlignment=true, WriteFacing=false)
   &ScaleColor(
    Scale Alpha=&Waveform(
     "Waveform [1]"=::NewEnumerator0
     "[1] Frequency"=&MergeHelper_FloatSwitch(0=0.8, 1=0.6, 2=0.7, 3=0.65, 4=0.75, 5=0.7, Index=$Particles.PartitionParticles.Partition, Num=::NewEnumerator4)
     "[1] Amplitude Min/Max"=<0 1>
    )
    Scale RGB=&MakeVectorFromLinearColorRGB(Color=<rgba 2 2 2 1>)
    ScaleA=true
    ScaleRGB=true
   )
   /SetVariables(Particles.Primary Color:LinearColor=&MakeLinearColorFromVectorAndFloat(
     "Vector (RGB)"=&MakeVectorFromLinearColorRGB(Color=&MultiplyLinearColor(
       Linear Color=&FNTimeOfDayColor(
        Night Color=<rgba 0.3148 0.1628 0.4531 1>
        Morning Color=<rgba 1 0.4336 0.3594 1>
        Day Color=<rgba 0.8906 0.2978 0.5362 1>
        Evening Color=<rgba 0.9115 0.3617 0.5433 1>
       )
       Scale Factor=&MakeVector4_FromLinearColor(LinearColor=<rgba 1 1 1 1>)
      ))
     "Float (Alpha)"=&Multiply_Float(
      A=&Multiply_Float(A=1, B=$User.FX_Alpha)
      B=&FloatFromCurve(
       FloatCurve=DataInterface() {propertyValues {Curve {keys [<at 0.4 1>,<at 0.8 0>]}}}
       CurveIndex=&Abs_Float(Float:NiagaraFloat=&DotProduct(
         Vector A=$Particles.SpriteAlignment
         Vector B=&ReturnCameraVector(Camera Vectors=::NewEnumerator1, Coordinate Space=::Simulation)
        ))
      )
     )
    )) {isSetParameters=true, module=SetVariables_#11}
   /SetVariables(
    Particles.Secondary Color:LinearColor=&MakeLinearColorFromVectorAndFloat(
     "Vector (RGB)"=&MakeVectorFromLinearColorRGB(Color=&MultiplyLinearColor(
       Linear Color=&FNTimeOfDayColor(
        Night Color=<rgba 0.3148 0.1628 0.4531 1>
        Morning Color=<rgba 1 0.4336 0.3594 1>
        Day Color=<rgba 0.8906 0.2978 0.5362 1>
        Evening Color=<rgba 0.9115 0.3617 0.5433 1>
       )
       Scale Factor=&MakeVector4_FromLinearColor(LinearColor=<rgba 1 1 1 1>)
      ))
     "Float (Alpha)"=&Multiply_Float(
      A=&Multiply_Float(A=1, B=$User.FX_Alpha)
      B=&FloatFromCurve(
       FloatCurve=DataInterface() {propertyValues {Curve {keys [<at 0.5 0>,<at 1 1>]}}}
       CurveIndex=&Abs_Float(Float:NiagaraFloat=&DotProduct(
         Vector A=$Particles.SpriteAlignment
         Vector B=&ReturnCameraVector(Camera Vectors=::NewEnumerator1, Coordinate Space=::Simulation)
        ))
      )
     )
    )
    Particles.Secondary SpriteAlignment:Vector3f=<0 0 1>
   ) {isSetParameters=true, module=SetVariables_#16}
   /SetVariables(Particles.Position=&AddVectorToPosition(
     Position=$Particles.Position
     Vector=&Multiply_VectorByFloat(
      Vector=&MergeHelper_VectorSwitch(
       0=&RandomRangeVector(Minimum=<-0.4 1.2 1>, Maximum=<-0.1 1.4 0.8>)
       1=&RandomRangeVector(Minimum=<-0.25 -0.5 -0.5>, Maximum=<0.25 -0.3 -0.7>)
       2=&RandomRangeVector(Minimum=<-0.25 -0.4 -0.6>, Maximum=<0.25 -0.5 -0.5>)
       3=&RandomRangeVector(Minimum=<-0.25 0.7 0.4>, Maximum=<0.25 1 0.6>)
       4=&RandomRangeVector(Minimum=<-0.25 0.4 0.6>, Maximum=<0.25 0.7 0.75>)
       5=&RandomRangeVector(Minimum=<-0.25 -1 -1>, Maximum=<0.25 1.25 1>, Evaluation Type=::NewEnumerator1)
       Index=$Particles.PartitionParticles.Partition
       Num=::NewEnumerator4
      )
      Float=&Waveform(
       "Waveform [1]"=::NewEnumerator0
       "[1] Frequency"=&RandomRangeFloat(Minimum=0.05, Maximum=0.13)
       "[1] Amplitude Min/Max"=&Multiply_Vector2D_ByFloat(Float=&RandomRangeFloat(Minimum=0.5, Maximum=0.7), Vector2D=<-1 1>)
      )
     )
    )) {enabled=false, isSetParameters=true, module=SetVariables_#17}
   /SetVariables(Particles.Position=&AddVectorToPosition(
     Position=$Particles.Position
     Vector=&Add_Vector(
      A=&Multiply_VectorByFloat(
       Vector=&MergeHelper_VectorSwitch(
        0=<-0.5 0 0>
        1=<0.5 0 0>
        2=<0.4 0 0>
        3=<0.5 0 0>
        4=<0.4 0 0>
        5=<0.15 0 0>
        Index=$Particles.Initial.PartitionParticles.Partition
        Num=::NewEnumerator4
       )
       Float=&Waveform(
        "Waveform [1]"=::NewEnumerator0
        "[1] Frequency"=&RandomRangeFloat(Minimum=0.08, Maximum=0.12)
        "[1] Amplitude Min/Max"=&Multiply_Vector2D_ByFloat(Float=&RandomRangeFloat(Minimum=0.5, Maximum=0.7), Vector2D=<-1 1>)
       )
      )
      B=&Multiply_VectorByFloat(
       Vector=&MergeHelper_VectorSwitch(
        0=&RandomRangeVector(Minimum=<0 1.2 1>, Maximum=<0 1.4 0.8>)
        1=&RandomRangeVector(Minimum=<0 -0.5 -0.5>, Maximum=<0 -0.3 -0.7>)
        2=&RandomRangeVector(Minimum=<0 -0.4 -0.6>, Maximum=<0 -0.5 -0.5>)
        3=&RandomRangeVector(Minimum=<0 0.7 0.4>, Maximum=<0 1 0.6>)
        4=&RandomRangeVector(Minimum=<0 0.4 0.6>, Maximum=<0 0.7 0.75>)
        5=&RandomRangeVector(Minimum=<0 -1 -1>, Maximum=<0 1.25 1>, Evaluation Type=::NewEnumerator1)
        Index=$Particles.PartitionParticles.Partition
        Num=::NewEnumerator4
       )
       Float=&Waveform(
        "Waveform [1]"=::NewEnumerator0
        "[1] Frequency"=&RandomRangeFloat(Minimum=0.05, Maximum=0.13)
        "[1] Amplitude Min/Max"=&Multiply_Vector2D_ByFloat(Float=&RandomRangeFloat(Minimum=0.5, Maximum=0.7), Vector2D=<-1 1>)
       )
      )
     )
    )) {isSetParameters=true, module=SetVariables_#12}
   /SetVariables(Particles.PivotOffset:Vector2f=&MakeVector2D(
     X=0.5
     Y=&Waveform(
      "Waveform [1]"=::NewEnumerator0
      "[1] Frequency"=&MergeHelper_FloatSwitch(0=0.8, 1=0.6, 2=0.7, 3=0.65, 4=0.75, 5=0.7, Index=$Particles.PartitionParticles.Partition, Num=::NewEnumerator4)
      "[1] Amplitude Min/Max"=<0.25 -0.25>
     )
    )) {isSetParameters=true, module=SetVariables_#13}
   &GravityForce(Gravity=&MaskVector(
     Mask Source=0
     Mask Parameter=?
     Mask=0
     Masked Vector=<0 0 0>
     Passthrough Vector=&MakeVector(X=0, Y=-50, Z=-50)
    ))
   /SetVariables(Particles.Bubbles Dynamic Param:Vector4f=&MakeVector4(
     X=&FloatFromCurve(FloatCurve=DataInterface() {propertyValues {Curve {keys [
          {arriveTangent=0.1706, interpMode=RCIM_Cubic, leaveTangent=0.1706, leaveTangentWeight=0.3081, tangentMode=RCTM_User, time=0, value=0}
          {arriveTangentWeight=0.3104, interpMode=RCIM_Cubic, leaveTangentWeight=0.0345, tangentMode=RCTM_User, time=0.9, value=0.2}
          {arriveTangent=3.2005, interpMode=RCIM_Cubic, leaveTangent=3.2005, tangentMode=RCTM_User, time=1, value=0.5}
         ]}}})
     Y=0
     Z=0
     W=0
    )) {isSetParameters=true, module=SetVariables_#14}
   &PointForce(
    Force Origin=&AddVectorToPosition(Position=&SimulationPosition(), Vector=<0 0 100>)
    Force Strength=&FloatFromCurve(FloatCurve=DataInterface() {propertyValues {Curve {keys [<at 0 0>,<at 0.1 1>]}}}, CurveIndex=$User.Jelly_Push, Scale Curve=500)
    Module.Use Falloff=false
    Module.UseLinearForce=false
   )
   &SolveForcesAndVelocity(Clamp Velocity=false, Limit Acceleration=false)
   /SetVariables(Particles.Vector 2D:Vector2f=&Multiply_Vector2D(
     A=&Vector2DFromFloat(Value=&MergeHelper_FloatSwitch(0=28.5, 1=25, 2=28, 3=25, 4=27, 5=25, Index=$Particles.PartitionParticles.Partition, Num=::NewEnumerator4))
     B=&Vector2DFromFloat(Value=&FloatFromCurve(
       FloatCurve=DataInterface() {propertyValues {Curve {keys [<at 0.5 1>,<at 1 0>]}}}
       CurveIndex=$User.Jelly_Push
       Scale Curve=&FloatFromCurve(FloatCurve=DataInterface() {propertyValues {Curve {keys [<at 0 1>]}}}, CurveIndex=$Particles.Age, Scale Curve=1.1)
      ))
    )) {isSetParameters=true, module=SetVariables_#15}
  ]
  rendererClasses [&NiagaraSpriteRendererProperties]
  renderers [
   {rendererClass=&NiagaraSpriteRendererProperties}
   {rendererClass=&NiagaraSpriteRendererProperties, rendererIndex=1}
  ]
 }
}
summary {
 systemName=NS_Companion_SpendPlume_Jellyfish
 emitters [
  {emitterName=JellyFish, rendererClasses [&NiagaraSpriteRendererProperties]}
  {emitterName=FXJelly_Bubbles, rendererClasses [&NiagaraSpriteRendererProperties]}
  {bEnabled=false, emitterName=BigJelly_Bubbles, rendererClasses [&NiagaraSpriteRendererProperties]}
  {emitterName=FXJelly_SpawnBubbles, rendererClasses [&NiagaraSpriteRendererProperties]}
  {emitterName=FXJelly_Bubbles001, rendererClasses [&NiagaraSpriteRendererProperties]}
  {emitterName=BubblePops, rendererClasses [&NiagaraSpriteRendererProperties]}
  {bEnabled=false, emitterName=JellyFish001, rendererClasses [&NiagaraSpriteRendererProperties]}
  {bEnabled=false, emitterName=JellyFishBackup, rendererClasses [&NiagaraSpriteRendererProperties]}
 ]
 userVariables [
  User.bIsFrontEndPreview {type=NiagaraBool}
  User.FX_Alpha {type=NiagaraFloat, defaultValue=1, description=""}
  User.Jelly_Push {type=NiagaraFloat, defaultValue=0, description=""}
  User.JellyFish_Timing {type=NiagaraFloat, defaultValue=0, description=""}
  User.JellySwitch {type=NiagaraFloat, defaultValue=1, description=""}
  User.OwnerLocalVelocity {type=Vector3f, description="", defaultValue {struct=Vector3f, value=<0 0 0>}}
 ]
}
